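# Niagara one-shot burst effect for the kick in the IndoorLaceBeat emote, throwing flickering sparkles, short flashes and tumbling meshes, mostly in purple-blue hues.
# Extra_Sparkles adds a small cone spray beside Main_Burst_Condensed and its GPU-sim copy, Main_Burst_Condensed001.
# Condensed emitters pack seven spawn groups into one emitter, using MergeHelper switches on SpawnGroup for per-layer look and motion, with VisibilityTag storing the group.
# The CPU burst's velocity-based world/local BlendSpace module is disabled.
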NiagaraSystem /BRCosmetics/Animation/Game/MainPlayer/Emotes/IndoorLaceBeat/FX/NS_Emote_IndoorLaceBeat_Kick
dependencies {
 usedDataInterfaces [
  &NiagaraDataInterfaceCurve
  &NiagaraDataInterfaceColorCurve
  &NiagaraDataInterfaceCamera
  &NiagaraDataInterfaceVectorCurve
  &NiagaraDataInterfaceArrayColor
  &NiagaraDataInterfaceArrayFloat
  &NiagaraDataInterfaceSpriteRendererInfo
 ]
 usedDynamicInputs []
 usedModules [
  &SystemState
  &BlendSpaceHelper
  &self:SystemSpawnScript.SystemScriptSource.SystemScriptGraph.NiagaraNodeAssignment_0.SetVariables_#1
 ]
}
emitters {
 Extra_Sparkles {
  emitterUpdateScript [
   &EmitterState()
   &SpawnBurst_Instantaneous(Spawn Count=11, Use Spawn Probability=false, UseLoopCountLimit=false)
  ]
  particleSpawnScript [
   /SetVariables(Particles.Scale:NiagaraFloat=&RandomRangeFloat(Minimum=0, Maximum=1)) {isSetParameters=true}
   &InitializeParticle(
    Color Mode=::NewEnumerator0
    Lifetime=&FloatFromCurve(
     FloatCurve=DataInterface() {propertyValues {Curve {keys [<at 0 0.6667 RCIM_Constant>,<at 0.4563 0.9333 RCIM_Constant>,<at 0.93 0.7333 RCIM_Constant>]}}}
     CurveIndex=$Particles.Scale
     Scale Curve=&RandomRangeFloat(Minimum=0.6, Maximum=1.1)
    )
    Sprite Size Mode=::NewEnumerator3
    Uniform Sprite Size=&FloatFromCurve(
     FloatCurve=DataInterface() {propertyValues {Curve {keys [<at 0 0.3 RCIM_Constant>,<at 0.4406 0.5 RCIM_Constant>,<at 0.921 2.1 RCIM_Constant>]}}}
     CurveIndex=$Particles.Scale
     Scale Curve=15
    )
    UsePositionOffset=false
    Write Material Random=true
   )
   &AddVelocity(
    Cone Angle=55
    Cone Axis=<0 1 0.6>
    Module.UseConeFalloff=false
    Module.UseVelDistribution=false
    Override Seed=false
    Velocity Mode=::NewEnumerator2
    Velocity Speed=&RandomRangeFloat(Minimum=150, Maximum=350)
   )
   &ApplyInitialForces()
   &ShapeLocation(MaskSpawnGroup=false, Override Seed=false, Sphere Radius=15, UseOwnerScale=true)
  ]
  particleUpdateScript [
   &ParticleState()
   &Drag(Drag=&RandomRangeFloat(Minimum=2, Maximum=5), Use Linear Drag=true, Use Rotational Drag=true)
   &SolveForcesAndVelocity(Clamp Velocity=false, Limit Acceleration=false)
   &ScaleColor(
    Scale Alpha=&FloatFromCurve(
     FloatCurve=DataInterface() {propertyValues {Curve {
        defaultValue=1
        keys [
         {interpMode=RCIM_Cubic, tangentMode=RCTM_User, time=0, value=0}
         {interpMode=RCIM_Cubic, tangentMode=RCTM_User, time=0.2269, value=1}
         {arriveTangent=-0.1174, interpMode=RCIM_Cubic, leaveTangent=-0.1174, tangentMode=RCTM_User, time=0.3787, value=1}
         {tangentMode=RCTM_User, time=1, value=0}
        ]
       }}}
     Scale Curve=&Waveform(Global Phase=$Particles.Age, "Waveform [1]"=::NewEnumerator0, "[1] Frequency"=5, "[1] Amplitude Min/Max"=<0 1>, "[1] Bias"=0.3)
    )
    ScaleA=true
    ScaleRGB=true
   )
  ]
  rendererClasses [&NiagaraSpriteRendererProperties]
  renderers [{rendererClass=&NiagaraSpriteRendererProperties}]
 }
 Main_Burst_Condensed {
  emitterUpdateScript [
   &EmitterState()
   /SetVariables(Emitter.WalkForwardSpeed=145) {isSetParameters=true, module=SetVariables_#2}
   &SpawnBurst_Instantaneous(Use Spawn Probability=false, UseLoopCountLimit=false)
   &SpawnBurst_Instantaneous(Spawn Group=1, Use Spawn Probability=false, UseLoopCountLimit=false) {module=SpawnBurst_Instantaneous005}
   &SpawnBurst_Instantaneous(Spawn Group=2, Use Spawn Probability=false, UseLoopCountLimit=false) {module=SpawnBurst_Instantaneous001}
   &SpawnBurst_Instantaneous(Spawn Count=5, Spawn Group=3, Use Spawn Probability=false, UseLoopCountLimit=false) {module=SpawnBurst_Instantaneous002}
   &SpawnBurst_Instantaneous(Spawn Count=6, Spawn Group=4, Use Spawn Probability=false, UseLoopCountLimit=false) {module=SpawnBurst_Instantaneous003}
   &SpawnBurst_Instantaneous(Spawn Count=7, Spawn Group=5, Use Spawn Probability=false, UseLoopCountLimit=false) {module=SpawnBurst_Instantaneous004}
   &SpawnBurst_Instantaneous(
    Spawn Count=&RandomRangeInt(Minimum=2, Maximum=3)
    Spawn Group=6
    Use Spawn Probability=false
    UseLoopCountLimit=false
   ) {module=SpawnBurst_Instantaneous006}
  ]
  particleSpawnScript [
   /SetVariables(Particles.VisibilityTag=&SpawnGroup()) {isSetParameters=true, module=SetVariables_#3}
   &InitializeParticle(
    Color=&MergeHelper_ColorSwitch(
     0=<rgba 0.017 0.0177 0.0323 2>
     1=<rgba 1 1 1 1>
     2=<rgba 1 1 1 1>
     3=&SelectLinearColorFromArray(
      Array Sampling Mode=::NewEnumerator0
      Direct Array Index=?
      Color Selection Array=DataInterface() {propertyValues {ColorData [
         <rgba 0.5658 0.1867 3 1>
         <rgba 0.7301 0.5794 0.9427 1>
         <rgba 0.3982 0.2695 0.7917 1>
         <rgba 0.1423 0.3438 0.8906 1>
         <rgba 0.6181 1.3757 3 1>
        ]}}
      Random Seed=0
     )
     4=&SelectLinearColorFromArray(
      Array Sampling Mode=::NewEnumerator0
      Direct Array Index=?
      Color Selection Array=DataInterface() {propertyValues {ColorData [
         <rgba 0.096 0.0283 0.7529 1>
         <rgba 0.0404 0.0967 1 1>
         <rgba 0.1473 0.1405 0.149 1>
         <rgba 0.1297 0.1437 0.3281 1>
         <rgba 0.1226 0.0724 0.2406 1>
         <rgba 0.3385 0.1618 1 1>
         <rgba 0.2146 0.3534 0.7698 1>
         <rgba 0.251 0.1312 0.4281 1>
        ]}}
      Random Seed=0
     )
     5=&MultiplyLinearColor(
      Linear Color=/Niagara/DynamicInputs/TypeConversions/MakeLinearColorFromFloat(Value:NiagaraFloat=&Lerp_Float(A=0.3, B=&RandomRangeFloat(Minimum=0, Maximum=1), Alpha=$System.Flicker))
      Scale Factor=&MakeVector4_FromLinearColor(LinearColor=&SelectLinearColorFromArray(
        Array Sampling Mode=::NewEnumerator0
        Direct Array Index=?
        Color Selection Array=DataInterface() {propertyValues {ColorData [<rgba 0.6685 0.1858 2 1>, <rgba 0.2468 0.3864 2 1>, <rgba 0.1207 0.3222 2 1>]}}
        Random Seed=0
       ))
     )
     6=&SelectLinearColorFromArray(
      Array Sampling Mode=::NewEnumerator0
      Direct Array Index=?
      Color Selection Array=DataInterface() {propertyValues {ColorData [<rgba 0.6525 0.5217 3 1>, <rgba 0.5217 1.3744 5 1>]}}
      Random Seed=0
     )
     Index=&SpawnGroup()
     Num=::NewEnumerator5
    )
    Lifetime=&MergeHelper_FloatSwitch(
     0=0.25
     1=0.15
     2=0.2
     3=&RandomRangeFloat(Minimum=0.3, Maximum=0.9)
     4=&RandomRangeFloat(Minimum=0.5, Maximum=1)
     5=&RandomRangeFloat(Minimum=0.3, Maximum=1.5)
     6=&RandomRangeFloat(Minimum=0.35, Maximum=1)
     Index=&SpawnGroup()
     Num=::NewEnumerator5
    )
    Mass=&MergeHelper_FloatSwitch(
     0=1
     1=1
     2=1
     3=1
     4=&RandomRangeFloat(Minimum=0.75, Maximum=2)
     5=1
     6=1
     Index=&SpawnGroup()
     Num=::NewEnumerator5
    )
    Mass Mode=::NewEnumerator1
    Mesh Scale Mode=::NewEnumerator3
    Mesh Uniform Scale=&RandomRangeFloat(Minimum=0.4, Maximum=1)
    Position Offset=&MaskVector(
     Mask Source=0
     Mask Parameter=?
     Mask=5
     Masked Vector=&RandomRangeVector(Minimum=<0 5 0>, Maximum=<0 9 0>)
     Passthrough Vector=<0 0 0>
    )
    Sprite Rotation Angle=&MergeHelper_FloatSwitch(
     0=0
     1=&SelectFloatFromArray(
      Array Sampling Mode=::NewEnumerator0
      Direct Array Index=?
      Float Selection Array=DataInterface() {propertyValues {FloatData [-22,22,-33,33]}}
      Random Seed=0
     )
     2=0
     3=0
     4=0
     5=&RandomRangeFloat(Minimum=-360, Maximum=360)
     Index=&SpawnGroup()
     Num=::NewEnumerator4
    )
    Sprite Rotation Mode=::NewEnumerator0
    Sprite Size Mode=::NewEnumerator3
    Uniform Sprite Size=&MergeHelper_FloatSwitch(
     0=80
     1=80
     2=80
     3=&RandomRangeFloat(Minimum=8, Maximum=20)
     4=&RandomRangeFloat(Minimum=1.5, Maximum=20)
     5=&RandomRangeFloat(Minimum=9, Maximum=15)
     6=20
     Index=&SpawnGroup()
     Num=::NewEnumerator5
    )
    UsePositionOffset=true
    Write Material Random=true
   )
   &ShapeLocation(
    MaskSpawnGroup=false
    Non Uniform Scale=&MergeHelper_VectorSwitch(0=<0 0 0>, 1=<0 0 0>, 2=<0 0 0>, 3=<1 1 1>, 4=<1 1 1>, 5=<1 0.2 1>, Index=&SpawnGroup(), Num=::NewEnumerator4)
    Offset=&MaskVector(Mask Source=0, Mask Parameter=?, Mask=5, Masked Vector=<0 4 0>, Passthrough Vector=<0 0 0>)
    Offset Mode=::NewEnumerator0
    Override Seed=false
    Sphere Radius=&MergeHelper_FloatSwitch(0=0, 1=0, 2=0, 3=10, 4=10, 5=30, Index=&SpawnGroup(), Num=::NewEnumerator4)
    UseOwnerScale=true
   )
   &AddVelocity(
    Cone Angle=&MergeHelper_FloatSwitch(0=0, 1=0, 2=0, 3=70, 4=50, 5=0, 6=60, Index=&SpawnGroup(), Num=::NewEnumerator5)
    Cone Axis=&MergeHelper_VectorSwitch(
     0=<0 0 0>
     1=<0 0 0>
     2=<0 0 0>
     3=<0 0.5 0.5>
     4=<0 1 1>
     5=<0 1 0.6>
     6=<0 1 1>
     Index=&SpawnGroup()
     Num=::NewEnumerator5
    )
    Distribution Along Cone Axis:NiagaraFloat=&MergeHelper_FloatSwitch(0=0, 1=0, 2=0, 3=0.5, 4=0.2, 5=0.5, 6=0, Index=&SpawnGroup(), Num=::NewEnumerator5)
    Module.UseConeFalloff=false
    Module.UseVelDistribution=true
    Override Seed=false
    Velocity Mode=::NewEnumerator2
    Velocity Speed=&MergeHelper_FloatSwitch(
     0=0
     1=0
     2=0
     3=&RandomRangeFloat(Minimum=250, Maximum=350)
     4=&RandomRangeFloat(Minimum=200, Maximum=390)
     5=&RandomRangeFloat(Minimum=30, Maximum=150)
     6=400
     Index=&SpawnGroup()
     Num=::NewEnumerator5
    )
   )
   &AddVelocity(Velocity=&MergeHelper_VectorSwitch(
     0=<0 0 0>
     1=<0 0 0>
     2=<0 0 0>
     3=<0 40 20>
     4=<0 50 0>
     5=&RandomRangeVector(Minimum=<0 20 0>, Maximum=<0 50 25>)
     6=<0 20 0>
     Index=&SpawnGroup()
     Num=::NewEnumerator5
    )) {module=AddVelocity001}
   &CameraOffset(Camera Offset Amount=&MergeHelper_FloatSwitch(0=3, 1=3, 2=3, 3=0, 4=0, 5=0, Index=&SpawnGroup(), Num=::NewEnumerator4))
   &InitialMeshOrientation(Use Rotation Vector=false)
   /SetVariables(Particles.ClockHandSize:Vector2f=<10 40>) {isSetParameters=true, module=SetVariables_#4}
   /SetVariables() {isSetParameters=true, module=SetVariables_#5}
  ]
  particleUpdateScript [
   &ParticleState()
   &ScaleMeshSize(Scale Factor=&MergeHelper_VectorSwitch(
     0=<0 0 0>
     1=<0 0 0>
     2=<0 0 0>
     3=<0 0 0>
     4=<0 0 0>
     5=<0 0 0>
     6=&VectorFromFloat(Value=&FloatFromCurve(FloatCurve=DataInterface() {propertyValues {Curve {
          defaultValue=1
          keys [
           {arriveTangent=9.0725, interpMode=RCIM_Cubic, leaveTangent=9.0725, tangentMode=RCTM_User, time=0, value=0}
           {arriveTangent=0.0953, interpMode=RCIM_Cubic, leaveTangent=0.0953, tangentMode=RCTM_User, time=0.0297, value=0.9847}
           {arriveTangent=-1.0148, interpMode=RCIM_Cubic, leaveTangent=-1.0148, time=0.3517, value=1}
           {arriveTangent=-0.0405, interpMode=RCIM_Cubic, leaveTangent=-0.0405, tangentMode=RCTM_User, time=1, value=0}
          ]
         }}}))
     Index=&SpawnGroup()
     Num=::NewEnumerator5
    ))
   &SpriteRotationRate(Rotation Rate=&MergeHelper_FloatSwitch(
     0=0
     1=0
     2=0
     3=&Multiply_Float(
      A=&SelectFloatFromArray(
       Array Sampling Mode=::NewEnumerator0
       Direct Array Index=?
       Float Selection Array=DataInterface() {propertyValues {FloatData [-1,1]}}
       Random Seed=0
      )
      B=&RandomRangeFloat(Minimum=50, Maximum=500)
     )
     4=&Multiply_Float(
      A=&SelectFloatFromArray(
       Array Sampling Mode=::NewEnumerator0
       Direct Array Index=?
       Float Selection Array=DataInterface() {propertyValues {FloatData [-1,1]}}
       Random Seed=0
      )
      B=&RandomRangeFloat(Minimum=50, Maximum=500)
     )
     5=0
     Index=&SpawnGroup()
     Num=::NewEnumerator4
    ))
   &Drag(
    Drag=&MergeHelper_FloatSwitch(
     0=0
     1=0
     2=0
     3=0
     4=&RandomRangeFloat(Minimum=1, Maximum=1.4)
     5=&RandomRangeFloat(Minimum=1, Maximum=2)
     6=1
     Index=&SpawnGroup()
     Num=::NewEnumerator5
    )
    Use Linear Drag=true
    Use Rotational Drag=false
   ) {module=Drag001}
   &GravityForce(Gravity=&MergeHelper_VectorSwitch(
     0=<0 0 0>
     1=<0 0 0>
     2=<0 0 0>
     3=&Multiply_VectorByFloat(
      Vector=&RandomRangeVector(Minimum=<0 0 -900>, Maximum=<0 0 -1200>)
      Float=&FloatFromCurve(FloatCurve=DataInterface() {propertyValues {Curve {
          defaultValue=1
          keys [{arriveTangent=7.301, interpMode=RCIM_Cubic, leaveTangent=7.301, tangentMode=RCTM_User, time=0, value=0}, <at 0.3717 1 RCIM_Cubic>]
         }}})
     )
     4=&Multiply_VectorByFloat(
      Vector=<0 0 -700>
      Float=&FloatFromCurve(FloatCurve=DataInterface() {propertyValues {Curve {defaultValue=1, keys [<at 0 0 RCIM_Cubic>,<at 0.5 1 RCIM_Cubic>]}}})
     )
     5=<0 0 -22>
     6=&RandomRangeVector(Minimum=<0 0 -980>, Maximum=<1 1 -1800>)
     Index=&SpawnGroup()
     Num=::NewEnumerator5
    ))
   &Drag(
    Drag=&MergeHelper_FloatSwitch(0=0, 1=0, 2=0, 3=0.75, 4=0, 5=0, 6=0, Index=&SpawnGroup(), Num=::NewEnumerator5)
    Use Linear Drag=true
    Use Rotational Drag=false
   )
   &ScaleSpriteSize(
    Scale Sprite Size Mode=::NewEnumerator0
    Uniform Scale Factor=&MergeHelper_FloatSwitch(
     0=&FloatFromCurve(FloatCurve=DataInterface() {propertyValues {Curve {
         defaultValue=0
         keys [
          {interpMode=RCIM_Cubic, leaveTangent=6.2608, tangentMode=RCTM_Break, time=0, value=0.2}
          {interpMode=RCIM_Cubic, leaveTangent=-0.5405, tangentMode=RCTM_Break, time=0.3774, value=1.3168}
          {interpMode=RCIM_Cubic, tangentMode=RCTM_Break, time=1, value=0}
         ]
        }}})
     1=&FloatFromCurve(FloatCurve=DataInterface() {propertyValues {Curve {
         defaultValue=0
         keys [
          {interpMode=RCIM_Cubic, leaveTangent=6.2608, tangentMode=RCTM_Break, time=0, value=0.2}
          {interpMode=RCIM_Cubic, leaveTangent=-0.5405, tangentMode=RCTM_Break, time=0.3774, value=1.3168}
          {interpMode=RCIM_Cubic, tangentMode=RCTM_Break, time=1, value=0}
         ]
        }}})
     2=&FloatFromCurve(FloatCurve=DataInterface() {propertyValues {Curve {
         defaultValue=0
         keys [
          {interpMode=RCIM_Cubic, leaveTangent=6.2608, tangentMode=RCTM_Break, time=0, value=0.2}
          {interpMode=RCIM_Cubic, leaveTangent=-0.5405, tangentMode=RCTM_Break, time=0.3774, value=1.3168}
          {interpMode=RCIM_Cubic, tangentMode=RCTM_Break, time=1, value=0}
         ]
        }}})
     3=&FloatFromCurve(FloatCurve=DataInterface() {propertyValues {Curve {keys [
          <at -0.0046 1.0113 RCIM_Cubic>
          {arriveTangent=-1.0036, interpMode=RCIM_Cubic, leaveTangent=-1.0036, time=0.7001, value=0.872}
          <at 1.0011 0.002>
         ]}}})
     4=0
     5=&FloatFromCurve(
      FloatCurve=DataInterface() {propertyValues {Curve {keys [
          {interpMode=RCIM_Cubic, tangentMode=RCTM_User, time=0, value=0}
          {interpMode=RCIM_Cubic, tangentMode=RCTM_User, time=0.3745, value=1}
          {tangentMode=RCTM_User, time=1, value=0}
         ]}}}
      Scale Curve=&RandomRangeFloat(Minimum=1, Maximum=2)
     )
     Index=&SpawnGroup()
     Num=::NewEnumerator4
    )
   )
   &ScaleColor(
    Scale Alpha=&MergeHelper_FloatSwitch(
     0=&FloatFromCurve(FloatCurve=DataInterface() {propertyValues {Curve {keys [<at -0.0046 1.0113 RCIM_Cubic>,<at 1.0011 0.002 RCIM_Cubic>]}}})
     1=&FloatFromCurve(FloatCurve=DataInterface() {propertyValues {Curve {keys [<at -0.0046 1.0113 RCIM_Cubic>,<at 1.0011 0.002 RCIM_Cubic>]}}})
     2=&FloatFromCurve(FloatCurve=DataInterface() {propertyValues {Curve {keys [<at -0.0046 1.0113 RCIM_Cubic>,<at 1.0011 0.002 RCIM_Cubic>]}}})
     3=&FloatFromCurve(FloatCurve=DataInterface() {propertyValues {Curve {keys [<at -0.0046 4 RCIM_Cubic>,<at 1.0011 0.002 RCIM_Cubic>]}}})
     4=&FloatFromCurve(
      FloatCurve=DataInterface() {propertyValues {Curve {keys [<at -0.0046 1.0113 RCIM_Cubic>,<at 1.0011 0.002 RCIM_Cubic>]}}}
      Scale Curve=&Waveform("Waveform [1]"=::NewEnumerator0, "[1] Frequency"=4, "[1] Amplitude Min/Max"=<0.1 1.1>)
     )
     5=&FloatFromCurve(
      FloatCurve=DataInterface() {propertyValues {Curve {keys [
          {tangentMode=RCTM_User, time=0, value=1}
          {arriveTangent=-0.0405, interpMode=RCIM_Cubic, leaveTangent=-0.0405, tangentMode=RCTM_User, time=0.8, value=1}
          {arriveTangent=-8.0089, interpMode=RCIM_Cubic, leaveTangent=-8.0089, tangentMode=RCTM_User, time=1, value=0}
         ]}}}
      Scale Curve=&Waveform(
       Global Input Position=&RandomRangeFloat(Minimum=0, Maximum=1)
       Global Phase=&Add_Float(A=$Particles.Age, B=&RandomRangeFloat(Minimum=0, Maximum=1))
       "Waveform [1]"=::NewEnumerator0
       "[1] Frequency"=&RandomRangeFloat(Minimum=3.618, Maximum=11.08)
       "[1] Amplitude Min/Max"=<0 1>
       "[1] Scale"=&FNTimeOfDayScalar(Night Scale=0.85, Morning Scale=0.93, Evening Scale=0.93)
      )
     )
     Index=&SpawnGroup()
     Num=::NewEnumerator4
    )
    Scale RGB=&MergeHelper_VectorSwitch(
     0=<1 1 1>
     1=<1 1 1>
     2=<1 1 1>
     3=&MakeVectorFromLinearColorRGB(Color=&ColorFromCurve(ColorCurve=DataInterface() {propertyValues {
         AlphaCurve {keys [<at 0 1 RCIM_Cubic>,<at 1 0 RCIM_Cubic>]}
         BlueCurve {keys [<at 0 4.0084>,<at 0.0115 2.0042>,<at 0.0325 4.0084>,<at 0.0426 6.7721>,<at 0.0536 1>,<at 0.2872 1 RCIM_Cubic>,<at 0.7525 1>]}
         GreenCurve {keys [<at 0 1.3859>,<at 0.0115 0.693>,<at 0.0325 1.3859>,<at 0.0426 3.1708>,<at 0.0536 1>,<at 0.2872 1 RCIM_Cubic>,<at 0.7525 1>]}
         RedCurve {keys [<at 0 10>,<at 0.0115 5>,<at 0.0325 10>,<at 0.0426 15>,<at 0.0536 1>,<at 0.2872 1 RCIM_Cubic>,<at 0.7525 1>]}
        }}))
     4=&MakeVectorFromLinearColorRGB(Color=&MultiplyLinearColor(
       Linear Color=&ColorFromCurve(ColorCurve=DataInterface() {propertyValues {
          AlphaCurve {keys []}
          BlueCurve {keys [<at 0 1 RCIM_Cubic>,<at 0.1015 1>,<at 0.118 50>,<at 0.1312 1>,<at 0.306 1>,<at 0.3225 1>,<at 0.3379 1>,<at 1 1 RCIM_Cubic>]}
          GreenCurve {keys [<at 0 1 RCIM_Cubic>,<at 0.1015 1>,<at 0.118 50>,<at 0.1312 1>,<at 0.306 1>,<at 0.3225 1>,<at 0.3379 1>,<at 1 1 RCIM_Cubic>]}
          RedCurve {keys [<at 0 1 RCIM_Cubic>,<at 0.1015 1>,<at 0.118 50>,<at 0.1312 1>,<at 0.306 1>,<at 0.3225 1>,<at 0.3379 1>,<at 1 1 RCIM_Cubic>]}
         }})
       Scale Factor=&MakeVector4_FromLinearColor(LinearColor=&LerpLinearColors(
         StartColor=<rgba 1 1 1 1>
         EndColor=<rgba 20 20 20 1>
         LerpFactor=&Waveform("Waveform [1]"=::NewEnumerator0, "[1] Frequency"=4, "[1] Amplitude Min/Max"=<0 1>)
        ))
      ))
     5=&VectorFromFloat(Value=&Multiply_Float(
       A=1
       B=&Lerp_Float(
        A=0.25
        B=22
        Alpha=&FloatFromCurve(
         FloatCurve=DataInterface() {propertyValues {Curve {keys [<at 0 1>,<at 0.28 0>,<at 0.3479 0>,<at 0.3882 0.12>,<at 0.4357 0>,<at 1 0>]}}}
         Scale Curve=20
        )
       )
      ))
     Index=&SpawnGroup()
     Num=::NewEnumerator4
    )
    ScaleA=true
    ScaleRGB=true
   )
   &CurlNoiseForce(
    Bias Noise=false
    Noise Frequency=&MergeHelper_FloatSwitch(
     0=0
     1=0
     2=0
     3=0
     4=&RandomRangeFloat(Minimum=30, Maximum=40)
     5=1
     Index=&SpawnGroup()
     Num=::NewEnumerator4
    )
    Noise Strength=&MergeHelper_FloatSwitch(
     0=0
     1=0
     2=0
     3=0
     4=100
     5=&RandomRangeFloat(Minimum=15, Maximum=33)
     Index=&SpawnGroup()
     Num=::NewEnumerator4
    )
    Pan Noise=false
   )
   /SetVariables(
    Particles.Pink Flare Color:LinearColor=&MakeLinearColorFromVectorAndFloat(
     "Vector (RGB)"=&MakeVectorFromLinearColorRGB(Color=<rgba 1.7287 1.0625 6 1>)
     "Float (Alpha)"=&FloatFromCurve(FloatCurve=DataInterface() {propertyValues {Curve {
         defaultValue=1
         keys [{interpMode=RCIM_Cubic, tangentMode=RCTM_User, time=0, value=1}, {interpMode=RCIM_Cubic, tangentMode=RCTM_User, time=1, value=0}]
        }}})
    )
    Particles.Pink Flare Size:Vector2f=&Vector2DFromFloat(Value=&Multiply_Float(
      A=50
      B=&FloatFromCurve(FloatCurve=DataInterface() {propertyValues {Curve {
          defaultValue=1
          keys [
           {arriveTangent=14.6816, interpMode=RCIM_Cubic, leaveTangent=14.6816, tangentMode=RCTM_User, time=0, value=0}
           {arriveTangent=-1.7058, interpMode=RCIM_Cubic, leaveTangent=-1.7058, tangentMode=RCTM_User, time=0.2501, value=1}
           {arriveTangent=-0.0405, interpMode=RCIM_Cubic, leaveTangent=-0.0405, tangentMode=RCTM_User, time=1, value=0}
          ]
         }}})
     ))
   ) {isSetParameters=true, module=SetVariables_#6}
   /SetVariables(
    Particles.Corona Color:LinearColor=&MakeLinearColorFromVectorAndFloat(
     "Vector (RGB)"=&MakeVectorFromLinearColorRGB(Color=&ColorFromCurve(ColorCurve=DataInterface() {propertyValues {
         AlphaCurve {keys [<at 0 1 RCIM_Cubic>,<at 1 0 RCIM_Cubic>]}
         BlueCurve {keys [<at 0.0153 1.9192 RCIM_Cubic>,<at 0.155 1>,<at 1 1 RCIM_Cubic>]}
         GreenCurve {keys [<at 0.0153 3.0652 RCIM_Cubic>,<at 0.155 0.8164>,<at 1 0.3552 RCIM_Cubic>]}
         RedCurve {keys [<at 0.0153 5 RCIM_Cubic>,<at 0.155 0.2727>,<at 1 0.5133 RCIM_Cubic>]}
        }}))
     "Float (Alpha)"=&FloatFromCurve(FloatCurve=DataInterface() {propertyValues {Curve {
         defaultValue=0
         keys [
          {interpMode=RCIM_Cubic, leaveTangent=18.0304, leaveTangentWeight=0.0371, tangentMode=RCTM_Break, time=0.095, value=1}
          {arriveTangent=-3.2936, interpMode=RCIM_Cubic, leaveTangent=-3.2936, tangentMode=RCTM_User, time=0.2062, value=1}
          <at 1 0 RCIM_Cubic>
         ]
        }}})
    )
    Particles.Corona Size:Vector2f=&Vector2DFromFloat(Value=&Multiply_Float(
      A=18
      B=&FloatFromCurve(FloatCurve=DataInterface() {propertyValues {Curve {
          defaultValue=1
          keys [{interpMode=RCIM_Cubic, tangentMode=RCTM_User, time=0, value=1}, {interpMode=RCIM_Cubic, tangentMode=RCTM_User, time=1, value=0}]
         }}})
     ))
   ) {isSetParameters=true, module=SetVariables_#7}
   /SetVariables(Particles.Blend Glow Color:LinearColor=&MakeLinearColorFromVectorAndFloat(
     "Vector (RGB)"=&MakeVectorFromLinearColorRGB(Color=<rgba 0.8606 0.1953 1 1>)
     "Float (Alpha)"=&FloatFromCurve(FloatCurve=DataInterface() {propertyValues {Curve {
         defaultValue=1
         keys [{interpMode=RCIM_Cubic, tangentMode=RCTM_User, time=0.0857, value=1}, {interpMode=RCIM_Cubic, tangentMode=RCTM_User, time=1, value=0}]
        }}})
    )) {isSetParameters=true, module=SetVariables_#8}
   /SetVariables(
    Particles.ClockHandSize:Vector2f=&Multiply_Vector2D(
     A=$Particles.Initial.ClockHandSize
     B=&MakeVector2D(X=&Waveform("Waveform [1]"=::NewEnumerator0, "[1] Frequency"=3, "[1] Amplitude Min/Max"=<-1 1>), Y=1)
    )
    Particles.ClockPartSize:Vector2f=&Multiply_Vector2D(
     A=$Particles.Initial.SpriteSize
     B=&MakeVector2D(X=&Waveform("Waveform [1]"=::NewEnumerator0, "[1] Frequency"=3, "[1] Amplitude Min/Max"=<-1 1>), Y=1)
    )
   ) {isSetParameters=true, module=SetVariables_#9}
   &MeshRotationForce(Rotation=&Multiply_Vector(
     A=&RandomRangeVector(Minimum=<-1 -1 -1>, Maximum=<1 1 1>)
     B=&VectorFromFloat(Value=&RandomRangeFloat(Minimum=50, Maximum=200))
    ))
   &SolveForcesAndVelocity(Clamp Velocity=false, Limit Acceleration=false)
   &ScaleSpriteSizeBySpeed(
    Max Scale Factor=&MergeHelper_Vector2Switch(0=<1 1>, 1=<1 1>, 2=<1 1>, 3=<1 1>, 4=<1 3>, 5=<1 1>, Index=&SpawnGroup(), Num=::NewEnumerator4)
    Min Scale Factor=&MergeHelper_Vector2Switch(0=<1 1>, 1=<1 1>, 2=<1 1>, 3=<1 1>, 4=<1 1>, 5=<1 1>, Index=&SpawnGroup(), Num=::NewEnumerator4)
    Sample Scale Factor By Curve=true
    Scale Factor Curve=DataInterface() {propertyValues {Curve {keys [
        {arriveTangent=2.865, interpMode=RCIM_Cubic, time=0.1247, value=-0.0006}
        {arriveTangent=4.1522, interpMode=RCIM_Cubic, leaveTangent=4.1522, tangentMode=RCTM_User, time=0.3948, value=0.4893}
        <at 0.9958 0.9974 RCIM_Cubic>
       ]}}}
   )
   &BlendSpace("Position Blend (World <-> Local)":NiagaraFloat=&MakeCustomFloatFromBool(
     Bool=&SetBoolByFloatComparison(A=&VectorLength(Vector=$User.OwnerLocalVelocity), B=$Emitter.WalkForwardSpeed)
     True Float=1
     False Float=&FloatFromCurve(FloatCurve=DataInterface() {propertyValues {Curve {keys [<at 0.6 1>,<at 0.8 0.25>]}}})
     Translation Contribution:NiagaraFloat=&MergeHelper_FloatSwitch(
      0=&MakeCustomFloatFromBool(
       Bool=&SetBoolByFloatComparison(A=&VectorLength(Vector=$User.OwnerLocalVelocity), B=$Emitter.WalkForwardSpeed)
       True Float=1
       False Float=&FloatFromCurve(FloatCurve=DataInterface() {propertyValues {Curve {keys [<at 0 1 RCIM_Cubic>,<at 0.5 0 RCIM_Cubic>]}}})
      )
      1=&MakeCustomFloatFromBool(
       Bool=&SetBoolByFloatComparison(A=&VectorLength(Vector=$User.OwnerLocalVelocity), B=$Emitter.WalkForwardSpeed)
       True Float=1
       False Float=&FloatFromCurve(FloatCurve=DataInterface() {propertyValues {Curve {keys [<at 0 1 RCIM_Cubic>,<at 0.5 0 RCIM_Cubic>]}}})
      )
      2=&MakeCustomFloatFromBool(
       Bool=&SetBoolByFloatComparison(A=&VectorLength(Vector=$User.OwnerLocalVelocity), B=$Emitter.WalkForwardSpeed)
       True Float=1
       False Float=&FloatFromCurve(FloatCurve=DataInterface() {propertyValues {Curve {keys [<at 0 1 RCIM_Cubic>,<at 0.5 0 RCIM_Cubic>]}}})
      )
      3=&MakeCustomFloatFromBool(
       Bool=&SetBoolByFloatComparison(A=&VectorLength(Vector=$User.OwnerLocalVelocity), B=$Emitter.WalkForwardSpeed)
       True Float=1
       False Float=0
      )
      4=&MakeCustomFloatFromBool(
       Bool=&SetBoolByFloatComparison(A=&VectorLength(Vector=$User.OwnerLocalVelocity), B=$Emitter.WalkForwardSpeed)
       True Float=1
       False Float=0
      )
      5=&MakeCustomFloatFromBool(
       Bool=&SetBoolByFloatComparison(A=&VectorLength(Vector=$User.OwnerLocalVelocity), B=$Emitter.WalkForwardSpeed)
       True Float=1
       False Float=&FloatFromCurve(FloatCurve=DataInterface() {propertyValues {Curve {keys [<at 0 1 RCIM_Cubic>,<at 0.5 0 RCIM_Cubic>]}}})
      )
      Index=&SpawnGroup()
      Num=::NewEnumerator4
     )
     Rotation Contribution:NiagaraFloat=&MergeHelper_FloatSwitch(
      0=&MakeCustomFloatFromBool(
       Bool=&SetBoolByFloatComparison(A=&VectorLength(Vector=$User.OwnerLocalVelocity), B=$Emitter.WalkForwardSpeed)
       True Float=1
       False Float=&FloatFromCurve(FloatCurve=DataInterface() {propertyValues {Curve {keys [<at 0 1 RCIM_Cubic>,<at 0.5 0 RCIM_Cubic>]}}})
      )
      1=&MakeCustomFloatFromBool(
       Bool=&SetBoolByFloatComparison(A=&VectorLength(Vector=$User.OwnerLocalVelocity), B=$Emitter.WalkForwardSpeed)
       True Float=1
       False Float=&FloatFromCurve(FloatCurve=DataInterface() {propertyValues {Curve {keys [<at 0 1 RCIM_Cubic>,<at 0.5 0 RCIM_Cubic>]}}})
      )
      2=&MakeCustomFloatFromBool(
       Bool=&SetBoolByFloatComparison(A=&VectorLength(Vector=$User.OwnerLocalVelocity), B=$Emitter.WalkForwardSpeed)
       True Float=1
       False Float=&FloatFromCurve(FloatCurve=DataInterface() {propertyValues {Curve {keys [<at 0 1 RCIM_Cubic>,<at 0.5 0 RCIM_Cubic>]}}})
      )
      3=&MakeCustomFloatFromBool(
       Bool=&SetBoolByFloatComparison(A=&VectorLength(Vector=$User.OwnerLocalVelocity), B=$Emitter.WalkForwardSpeed)
       True Float=1
       False Float=0
      )
      4=&MakeCustomFloatFromBool(
       Bool=&SetBoolByFloatComparison(A=&VectorLength(Vector=$User.OwnerLocalVelocity), B=$Emitter.WalkForwardSpeed)
       True Float=1
       False Float=0
      )
      5=&MakeCustomFloatFromBool(
       Bool=&SetBoolByFloatComparison(A=&VectorLength(Vector=$User.OwnerLocalVelocity), B=$Emitter.WalkForwardSpeed)
       True Float=1
       False Float=&FloatFromCurve(FloatCurve=DataInterface() {propertyValues {Curve {keys [<at 0 1 RCIM_Cubic>,<at 0.5 0 RCIM_Cubic>]}}})
      )
      Index=&SpawnGroup()
      Num=::NewEnumerator4
     )
    )) {enabled=false}
   &SubUVAnimation(
    Direct Frame Index:NiagaraFloat=&RandomRangeFloat(Minimum=0, Maximum=4)
    "Sprite Renderer "=DataInterface() {propertyValues {SpriteRenderer=&self:Main_Burst_Condensed_0.NiagaraSpriteRendererProperties_6}}
    SubUV Animation Mode:ENiagara_SubUVLookupModeV2=::NewEnumerator5
   )
   &SubUVAnimation(
    Direct Frame Index:NiagaraFloat=&RandomRangeFloat(Minimum=0, Maximum=6)
    "Sprite Renderer "=DataInterface() {propertyValues {SpriteRenderer=&self:Main_Burst_Condensed_0.NiagaraSpriteRendererProperties_5}}
    SubUV Animation Mode:ENiagara_SubUVLookupModeV2=::NewEnumerator5
   ) {module=SubUVAnimation001}
  ]
  rendererClasses [&NiagaraSpriteRendererProperties, &NiagaraMeshRendererProperties]
  renderers [
   {rendererClass=&NiagaraSpriteRendererProperties}
   {rendererClass=&NiagaraSpriteRendererProperties, rendererIndex=1}
   {rendererClass=&NiagaraSpriteRendererProperties, rendererIndex=2}
   {rendererClass=&NiagaraSpriteRendererProperties, rendererIndex=3}
   {rendererClass=&NiagaraSpriteRendererProperties, rendererIndex=4}
   {rendererClass=&NiagaraSpriteRendererProperties, rendererIndex=5}
   {rendererClass=&NiagaraSpriteRendererProperties, rendererIndex=6}
   {rendererClass=&NiagaraMeshRendererProperties, rendererIndex=7}
  ]
 }
 Main_Burst_Condensed001 {
  simTarget=GPUComputeSim
  emitterUpdateScript [
   &EmitterState()
   /SetVariables(Emitter.WalkForwardSpeed=145) {isSetParameters=true, module=SetVariables_#2}
   &SpawnBurst_Instantaneous(Use Spawn Probability=false, UseLoopCountLimit=false)
   &SpawnBurst_Instantaneous(Spawn Group=1, Use Spawn Probability=false, UseLoopCountLimit=false) {module=SpawnBurst_Instantaneous005}
   &SpawnBurst_Instantaneous(Spawn Group=2, Use Spawn Probability=false, UseLoopCountLimit=false) {module=SpawnBurst_Instantaneous001}
   &SpawnBurst_Instantaneous(Spawn Count=5, Spawn Group=3, Use Spawn Probability=false, UseLoopCountLimit=false) {module=SpawnBurst_Instantaneous002}
   &SpawnBurst_Instantaneous(Spawn Count=6, Spawn Group=4, Use Spawn Probability=false, UseLoopCountLimit=false) {module=SpawnBurst_Instantaneous003}
   &SpawnBurst_Instantaneous(Spawn Count=7, Spawn Group=5, Use Spawn Probability=false, UseLoopCountLimit=false) {module=SpawnBurst_Instantaneous004}
   &SpawnBurst_Instantaneous(
    Spawn Count=&RandomRangeInt(Minimum=2, Maximum=3)
    Spawn Group=6
    Use Spawn Probability=false
    UseLoopCountLimit=false
   ) {module=SpawnBurst_Instantaneous006}
  ]
  particleSpawnScript [
   /SetVariables(Particles.VisibilityTag=&SpawnGroup()) {isSetParameters=true, module=SetVariables_#3}
   &InitializeParticle(
    Color=&MergeHelper_ColorSwitch(
     0=<rgba 0.017 0.0177 0.0323 2>
     1=<rgba 1 1 1 1>
     2=<rgba 1 1 1 1>
     3=&SelectLinearColorFromArray(
      Array Sampling Mode=::NewEnumerator0
      Direct Array Index=?
      Color Selection Array=DataInterface() {propertyValues {ColorData [
         <rgba 0.5658 0.1867 3 1>
         <rgba 0.7301 0.5794 0.9427 1>
         <rgba 0.3982 0.2695 0.7917 1>
         <rgba 0.1423 0.3438 0.8906 1>
         <rgba 0.6181 1.3757 3 1>
        ]}}
      Random Seed=0
     )
     4=&SelectLinearColorFromArray(
      Array Sampling Mode=::NewEnumerator0
      Direct Array Index=?
      Color Selection Array=DataInterface() {propertyValues {ColorData [
         <rgba 0.096 0.0283 0.7529 1>
         <rgba 0.0404 0.0967 1 1>
         <rgba 0.1473 0.1405 0.149 1>
         <rgba 0.1297 0.1437 0.3281 1>
         <rgba 0.1226 0.0724 0.2406 1>
         <rgba 0.3385 0.1618 1 1>
         <rgba 0.2146 0.3534 0.7698 1>
         <rgba 0.251 0.1312 0.4281 1>
        ]}}
      Random Seed=0
     )
     5=&MultiplyLinearColor(
      Linear Color=&MakeLinearColorFromFloat(Value:NiagaraFloat=&Lerp_Float(A=0.3, B=&RandomRangeFloat(Minimum=0, Maximum=1), Alpha=$System.Flicker))
      Scale Factor=&MakeVector4_FromLinearColor(LinearColor=&SelectLinearColorFromArray(
        Array Sampling Mode=::NewEnumerator0
        Direct Array Index=?
        Color Selection Array=DataInterface() {propertyValues {ColorData [<rgba 0.6685 0.1858 2 1>, <rgba 0.2468 0.3864 2 1>, <rgba 0.1207 0.3222 2 1>]}}
        Random Seed=0
       ))
     )
     6=&SelectLinearColorFromArray(
      Array Sampling Mode=::NewEnumerator0
      Direct Array Index=?
      Color Selection Array=DataInterface() {propertyValues {ColorData [<rgba 0.6525 0.5217 3 1>, <rgba 0.5217 1.3744 5 1>]}}
      Random Seed=0
     )
     Index=&SpawnGroup()
     Num=::NewEnumerator5
    )
    Lifetime=&MergeHelper_FloatSwitch(
     0=0.25
     1=0.15
     2=0.2
     3=&RandomRangeFloat(Minimum=0.3, Maximum=0.9)
     4=&RandomRangeFloat(Minimum=0.5, Maximum=1)
     5=&RandomRangeFloat(Minimum=0.3, Maximum=1.5)
     6=&RandomRangeFloat(Minimum=0.35, Maximum=1)
     Index=&SpawnGroup()
     Num=::NewEnumerator5
    )
    Mass=&MergeHelper_FloatSwitch(
     0=1
     1=1
     2=1
     3=1
     4=&RandomRangeFloat(Minimum=0.75, Maximum=2)
     5=1
     6=1
     Index=&SpawnGroup()
     Num=::NewEnumerator5
    )
    Mass Mode=::NewEnumerator1
    Mesh Scale Mode=::NewEnumerator3
    Mesh Uniform Scale=&RandomRangeFloat(Minimum=0.4, Maximum=1)
    Position Offset=&MaskVector(
     Mask Source=0
     Mask Parameter=?
     Mask=5
     Masked Vector=&RandomRangeVector(Minimum=<0 5 0>, Maximum=<0 9 0>)
     Passthrough Vector=<0 0 0>
    )
    Sprite Rotation Angle=&MergeHelper_FloatSwitch(
     0=0
     1=&SelectFloatFromArray(
      Array Sampling Mode=::NewEnumerator0
      Direct Array Index=?
      Float Selection Array=DataInterface() {propertyValues {FloatData [-22,22,-33,33]}}
      Random Seed=0
     )
     2=0
     3=0
     4=0
     5=&RandomRangeFloat(Minimum=-360, Maximum=360)
     Index=&SpawnGroup()
     Num=::NewEnumerator4
    )
    Sprite Rotation Mode=::NewEnumerator0
    Sprite Size Mode=::NewEnumerator3
    Uniform Sprite Size=&MergeHelper_FloatSwitch(
     0=80
     1=80
     2=80
     3=&RandomRangeFloat(Minimum=8, Maximum=20)
     4=&RandomRangeFloat(Minimum=1.5, Maximum=20)
     5=&RandomRangeFloat(Minimum=9, Maximum=15)
     6=20
     Index=&SpawnGroup()
     Num=::NewEnumerator5
    )
    UsePositionOffset=true
    Write Material Random=true
   )
   &ShapeLocation(
    MaskSpawnGroup=false
    Non Uniform Scale=&MergeHelper_VectorSwitch(0=<0 0 0>, 1=<0 0 0>, 2=<0 0 0>, 3=<1 1 1>, 4=<1 1 1>, 5=<1 0.2 1>, Index=&SpawnGroup(), Num=::NewEnumerator4)
    Offset=&MaskVector(Mask Source=0, Mask Parameter=?, Mask=5, Masked Vector=<0 4 0>, Passthrough Vector=<0 0 0>)
    Offset Mode=::NewEnumerator0
    Override Seed=false
    Sphere Radius=&MergeHelper_FloatSwitch(0=0, 1=0, 2=0, 3=10, 4=10, 5=30, Index=&SpawnGroup(), Num=::NewEnumerator4)
    UseOwnerScale=true
   )
   &AddVelocity(
    Cone Angle=&MergeHelper_FloatSwitch(0=0, 1=0, 2=0, 3=70, 4=50, 5=0, 6=60, Index=&SpawnGroup(), Num=::NewEnumerator5)
    Cone Axis=&MergeHelper_VectorSwitch(
     0=<0 0 0>
     1=<0 0 0>
     2=<0 0 0>
     3=<0 0.5 0.5>
     4=<0 1 1>
     5=<0 1 0.6>
     6=<0 1 1>
     Index=&SpawnGroup()
     Num=::NewEnumerator5
    )
    Distribution Along Cone Axis:NiagaraFloat=&MergeHelper_FloatSwitch(0=0, 1=0, 2=0, 3=0.5, 4=0.2, 5=0.5, 6=0, Index=&SpawnGroup(), Num=::NewEnumerator5)
    Module.UseConeFalloff=false
    Module.UseVelDistribution=true
    Override Seed=false
    Velocity Mode=::NewEnumerator2
    Velocity Speed=&MergeHelper_FloatSwitch(
     0=0
     1=0
     2=0
     3=&RandomRangeFloat(Minimum=250, Maximum=350)
     4=&RandomRangeFloat(Minimum=200, Maximum=390)
     5=&RandomRangeFloat(Minimum=30, Maximum=150)
     6=400
     Index=&SpawnGroup()
     Num=::NewEnumerator5
    )
   )
   &AddVelocity(Velocity=&MergeHelper_VectorSwitch(
     0=<0 0 0>
     1=<0 0 0>
     2=<0 0 0>
     3=<0 40 20>
     4=<0 50 0>
     5=&RandomRangeVector(Minimum=<0 20 0>, Maximum=<0 50 25>)
     6=<0 20 0>
     Index=&SpawnGroup()
     Num=::NewEnumerator5
    )) {module=AddVelocity001}
   &CameraOffset(Camera Offset Amount=&MergeHelper_FloatSwitch(0=3, 1=3, 2=3, 3=0, 4=0, 5=0, Index=&SpawnGroup(), Num=::NewEnumerator4))
   &InitialMeshOrientation(Use Rotation Vector=false)
   /SetVariables(Particles.ClockHandSize:Vector2f=<10 40>) {isSetParameters=true, module=SetVariables_#4}
   /SetVariables() {isSetParameters=true, module=SetVariables_#5}
  ]
  particleUpdateScript [
   &ParticleState()
   &ScaleMeshSize(Scale Factor=&MergeHelper_VectorSwitch(
     0=<0 0 0>
     1=<0 0 0>
     2=<0 0 0>
     3=<0 0 0>
     4=<0 0 0>
     5=<0 0 0>
     6=&VectorFromFloat(Value=&FloatFromCurve(FloatCurve=DataInterface() {propertyValues {Curve {
          defaultValue=1
          keys [
           {arriveTangent=9.0725, interpMode=RCIM_Cubic, leaveTangent=9.0725, tangentMode=RCTM_User, time=0, value=0}
           {arriveTangent=0.0953, interpMode=RCIM_Cubic, leaveTangent=0.0953, tangentMode=RCTM_User, time=0.0297, value=0.9847}
           {arriveTangent=-1.0148, interpMode=RCIM_Cubic, leaveTangent=-1.0148, time=0.3517, value=1}
           {arriveTangent=-0.0405, interpMode=RCIM_Cubic, leaveTangent=-0.0405, tangentMode=RCTM_User, time=1, value=0}
          ]
         }}}))
     Index=&SpawnGroup()
     Num=::NewEnumerator5
    ))
   &SpriteRotationRate(Rotation Rate=&MergeHelper_FloatSwitch(
     0=0
     1=0
     2=0
     3=&Multiply_Float(
      A=&SelectFloatFromArray(
       Array Sampling Mode=::NewEnumerator0
       Direct Array Index=?
       Float Selection Array=DataInterface() {propertyValues {FloatData [-1,1]}}
       Random Seed=0
      )
      B=&RandomRangeFloat(Minimum=50, Maximum=500)
     )
     4=&Multiply_Float(
      A=&SelectFloatFromArray(
       Array Sampling Mode=::NewEnumerator0
       Direct Array Index=?
       Float Selection Array=DataInterface() {propertyValues {FloatData [-1,1]}}
       Random Seed=0
      )
      B=&RandomRangeFloat(Minimum=50, Maximum=500)
     )
     5=0
     Index=&SpawnGroup()
     Num=::NewEnumerator4
    ))
   &Drag(
    Drag=&MergeHelper_FloatSwitch(
     0=0
     1=0
     2=0
     3=0
     4=&RandomRangeFloat(Minimum=1, Maximum=1.4)
     5=&RandomRangeFloat(Minimum=1, Maximum=2)
     6=1
     Index=&SpawnGroup()
     Num=::NewEnumerator5
    )
    Use Linear Drag=true
    Use Rotational Drag=false
   ) {module=Drag001}
   &GravityForce(Gravity=&MergeHelper_VectorSwitch(
     0=<0 0 0>
     1=<0 0 0>
     2=<0 0 0>
     3=&Multiply_VectorByFloat(
      Vector=&RandomRangeVector(Minimum=<0 0 -900>, Maximum=<0 0 -1200>)
      Float=&FloatFromCurve(FloatCurve=DataInterface() {propertyValues {Curve {
          defaultValue=1
          keys [{arriveTangent=7.301, interpMode=RCIM_Cubic, leaveTangent=7.301, tangentMode=RCTM_User, time=0, value=0}, <at 0.3717 1 RCIM_Cubic>]
         }}})
     )
     4=&Multiply_VectorByFloat(
      Vector=<0 0 -700>
      Float=&FloatFromCurve(FloatCurve=DataInterface() {propertyValues {Curve {defaultValue=1, keys [<at 0 0 RCIM_Cubic>,<at 0.5 1 RCIM_Cubic>]}}})
     )
     5=<0 0 -22>
     6=&RandomRangeVector(Minimum=<0 0 -980>, Maximum=<1 1 -1800>)
     Index=&SpawnGroup()
     Num=::NewEnumerator5
    ))
   &Drag(
    Drag=&MergeHelper_FloatSwitch(0=0, 1=0, 2=0, 3=0.75, 4=0, 5=0, 6=0, Index=&SpawnGroup(), Num=::NewEnumerator5)
    Use Linear Drag=true
    Use Rotational Drag=false
   )
   &ScaleSpriteSize(
    Scale Sprite Size Mode=::NewEnumerator0
    Uniform Scale Factor=&MergeHelper_FloatSwitch(
     0=&FloatFromCurve(FloatCurve=DataInterface() {propertyValues {Curve {
         defaultValue=0
         keys [
          {interpMode=RCIM_Cubic, leaveTangent=6.2608, tangentMode=RCTM_Break, time=0, value=0.2}
          {interpMode=RCIM_Cubic, leaveTangent=-0.5405, tangentMode=RCTM_Break, time=0.3774, value=1.3168}
          {interpMode=RCIM_Cubic, tangentMode=RCTM_Break, time=1, value=0}
         ]
        }}})
     1=&FloatFromCurve(FloatCurve=DataInterface() {propertyValues {Curve {
         defaultValue=0
         keys [
          {interpMode=RCIM_Cubic, leaveTangent=6.2608, tangentMode=RCTM_Break, time=0, value=0.2}
          {interpMode=RCIM_Cubic, leaveTangent=-0.5405, tangentMode=RCTM_Break, time=0.3774, value=1.3168}
          {interpMode=RCIM_Cubic, tangentMode=RCTM_Break, time=1, value=0}
         ]
        }}})
     2=&FloatFromCurve(FloatCurve=DataInterface() {propertyValues {Curve {
         defaultValue=0
         keys [
          {interpMode=RCIM_Cubic, leaveTangent=6.2608, tangentMode=RCTM_Break, time=0, value=0.2}
          {interpMode=RCIM_Cubic, leaveTangent=-0.5405, tangentMode=RCTM_Break, time=0.3774, value=1.3168}
          {interpMode=RCIM_Cubic, tangentMode=RCTM_Break, time=1, value=0}
         ]
        }}})
     3=&FloatFromCurve(FloatCurve=DataInterface() {propertyValues {Curve {keys [
          <at -0.0046 1.0113 RCIM_Cubic>
          {arriveTangent=-1.0036, interpMode=RCIM_Cubic, leaveTangent=-1.0036, time=0.7001, value=0.872}
          <at 1.0011 0.002>
         ]}}})
     4=0
     5=&FloatFromCurve(
      FloatCurve=DataInterface() {propertyValues {Curve {keys [
          {interpMode=RCIM_Cubic, tangentMode=RCTM_User, time=0, value=0}
          {interpMode=RCIM_Cubic, tangentMode=RCTM_User, time=0.3745, value=1}
          {tangentMode=RCTM_User, time=1, value=0}
         ]}}}
      Scale Curve=&RandomRangeFloat(Minimum=1, Maximum=2)
     )
     Index=&SpawnGroup()
     Num=::NewEnumerator4
    )
   )
   &ScaleColor(
    Scale Alpha=&MergeHelper_FloatSwitch(
     0=&FloatFromCurve(FloatCurve=DataInterface() {propertyValues {Curve {keys [<at -0.0046 1.0113 RCIM_Cubic>,<at 1.0011 0.002 RCIM_Cubic>]}}})
     1=&FloatFromCurve(FloatCurve=DataInterface() {propertyValues {Curve {keys [<at -0.0046 1.0113 RCIM_Cubic>,<at 1.0011 0.002 RCIM_Cubic>]}}})
     2=&FloatFromCurve(FloatCurve=DataInterface() {propertyValues {Curve {keys [<at -0.0046 1.0113 RCIM_Cubic>,<at 1.0011 0.002 RCIM_Cubic>]}}})
     3=&FloatFromCurve(FloatCurve=DataInterface() {propertyValues {Curve {keys [<at -0.0046 4 RCIM_Cubic>,<at 1.0011 0.002 RCIM_Cubic>]}}})
     4=&FloatFromCurve(
      FloatCurve=DataInterface() {propertyValues {Curve {keys [<at -0.0046 1.0113 RCIM_Cubic>,<at 1.0011 0.002 RCIM_Cubic>]}}}
      Scale Curve=&Waveform("Waveform [1]"=::NewEnumerator0, "[1] Frequency"=4, "[1] Amplitude Min/Max"=<0.1 1.1>)
     )
     5=&FloatFromCurve(
      FloatCurve=DataInterface() {propertyValues {Curve {keys [
          {tangentMode=RCTM_User, time=0, value=1}
          {arriveTangent=-0.0405, interpMode=RCIM_Cubic, leaveTangent=-0.0405, tangentMode=RCTM_User, time=0.8, value=1}
          {arriveTangent=-8.0089, interpMode=RCIM_Cubic, leaveTangent=-8.0089, tangentMode=RCTM_User, time=1, value=0}
         ]}}}
      Scale Curve=&Waveform(
       Global Input Position=&RandomRangeFloat(Minimum=0, Maximum=1)
       Global Phase=&Add_Float(A=$Particles.Age, B=&RandomRangeFloat(Minimum=0, Maximum=1))
       "Waveform [1]"=::NewEnumerator0
       "[1] Frequency"=&RandomRangeFloat(Minimum=3.618, Maximum=11.08)
       "[1] Amplitude Min/Max"=<0 1>
       "[1] Scale"=&FNTimeOfDayScalar(Night Scale=0.85, Morning Scale=0.93, Evening Scale=0.93)
      )
     )
     Index=&SpawnGroup()
     Num=::NewEnumerator4
    )
    Scale RGB=&MergeHelper_VectorSwitch(
     0=<1 1 1>
     1=<1 1 1>
     2=<1 1 1>
     3=&MakeVectorFromLinearColorRGB(Color=&ColorFromCurve(ColorCurve=DataInterface() {propertyValues {
         AlphaCurve {keys [<at 0 1 RCIM_Cubic>,<at 1 0 RCIM_Cubic>]}
         BlueCurve {keys [<at 0 4.0084>,<at 0.0115 2.0042>,<at 0.0325 4.0084>,<at 0.0426 6.7721>,<at 0.0536 1>,<at 0.2872 1 RCIM_Cubic>,<at 0.7525 1>]}
         GreenCurve {keys [<at 0 1.3859>,<at 0.0115 0.693>,<at 0.0325 1.3859>,<at 0.0426 3.1708>,<at 0.0536 1>,<at 0.2872 1 RCIM_Cubic>,<at 0.7525 1>]}
         RedCurve {keys [<at 0 10>,<at 0.0115 5>,<at 0.0325 10>,<at 0.0426 15>,<at 0.0536 1>,<at 0.2872 1 RCIM_Cubic>,<at 0.7525 1>]}
        }}))
     4=&MakeVectorFromLinearColorRGB(Color=&MultiplyLinearColor(
       Linear Color=&ColorFromCurve(ColorCurve=DataInterface() {propertyValues {
          AlphaCurve {keys []}
          BlueCurve {keys [<at 0 1 RCIM_Cubic>,<at 0.1015 1>,<at 0.118 50>,<at 0.1312 1>,<at 0.306 1>,<at 0.3225 1>,<at 0.3379 1>,<at 1 1 RCIM_Cubic>]}
          GreenCurve {keys [<at 0 1 RCIM_Cubic>,<at 0.1015 1>,<at 0.118 50>,<at 0.1312 1>,<at 0.306 1>,<at 0.3225 1>,<at 0.3379 1>,<at 1 1 RCIM_Cubic>]}
          RedCurve {keys [<at 0 1 RCIM_Cubic>,<at 0.1015 1>,<at 0.118 50>,<at 0.1312 1>,<at 0.306 1>,<at 0.3225 1>,<at 0.3379 1>,<at 1 1 RCIM_Cubic>]}
         }})
       Scale Factor=&MakeVector4_FromLinearColor(LinearColor=&LerpLinearColors(
         StartColor=<rgba 1 1 1 1>
         EndColor=<rgba 20 20 20 1>
         LerpFactor=&Waveform("Waveform [1]"=::NewEnumerator0, "[1] Frequency"=4, "[1] Amplitude Min/Max"=<0 1>)
        ))
      ))
     5=&VectorFromFloat(Value=&Multiply_Float(
       A=1
       B=&Lerp_Float(
        A=0.25
        B=22
        Alpha=&FloatFromCurve(
         FloatCurve=DataInterface() {propertyValues {Curve {keys [<at 0 1>,<at 0.28 0>,<at 0.3479 0>,<at 0.3882 0.12>,<at 0.4357 0>,<at 1 0>]}}}
         Scale Curve=20
        )
       )
      ))
     Index=&SpawnGroup()
     Num=::NewEnumerator4
    )
    ScaleA=true
    ScaleRGB=true
   )
   &CurlNoiseForce(
    Bias Noise=false
    Noise Frequency=&MergeHelper_FloatSwitch(
     0=0
     1=0
     2=0
     3=0
     4=&RandomRangeFloat(Minimum=30, Maximum=40)
     5=1
     Index=&SpawnGroup()
     Num=::NewEnumerator4
    )
    Noise Strength=&MergeHelper_FloatSwitch(
     0=0
     1=0
     2=0
     3=0
     4=100
     5=&RandomRangeFloat(Minimum=15, Maximum=33)
     Index=&SpawnGroup()
     Num=::NewEnumerator4
    )
    Pan Noise=false
   )
   /SetVariables(
    Particles.Pink Flare Color:LinearColor=&MakeLinearColorFromVectorAndFloat(
     "Vector (RGB)"=&MakeVectorFromLinearColorRGB(Color=<rgba 1.7287 1.0625 6 1>)
     "Float (Alpha)"=&FloatFromCurve(FloatCurve=DataInterface() {propertyValues {Curve {
         defaultValue=1
         keys [{interpMode=RCIM_Cubic, tangentMode=RCTM_User, time=0, value=1}, {interpMode=RCIM_Cubic, tangentMode=RCTM_User, time=1, value=0}]
        }}})
    )
    Particles.Pink Flare Size:Vector2f=&Vector2DFromFloat(Value=&Multiply_Float(
      A=50
      B=&FloatFromCurve(FloatCurve=DataInterface() {propertyValues {Curve {
          defaultValue=1
          keys [
           {arriveTangent=14.6816, interpMode=RCIM_Cubic, leaveTangent=14.6816, tangentMode=RCTM_User, time=0, value=0}
           {arriveTangent=-1.7058, interpMode=RCIM_Cubic, leaveTangent=-1.7058, tangentMode=RCTM_User, time=0.2501, value=1}
           {arriveTangent=-0.0405, interpMode=RCIM_Cubic, leaveTangent=-0.0405, tangentMode=RCTM_User, time=1, value=0}
          ]
         }}})
     ))
   ) {isSetParameters=true, module=SetVariables_#6}
   /SetVariables(
    Particles.Corona Color:LinearColor=&MakeLinearColorFromVectorAndFloat(
     "Vector (RGB)"=&MakeVectorFromLinearColorRGB(Color=&ColorFromCurve(ColorCurve=DataInterface() {propertyValues {
         AlphaCurve {keys [<at 0 1 RCIM_Cubic>,<at 1 0 RCIM_Cubic>]}
         BlueCurve {keys [<at 0.0153 1.9192 RCIM_Cubic>,<at 0.155 1>,<at 1 1 RCIM_Cubic>]}
         GreenCurve {keys [<at 0.0153 3.0652 RCIM_Cubic>,<at 0.155 0.8164>,<at 1 0.3552 RCIM_Cubic>]}
         RedCurve {keys [<at 0.0153 5 RCIM_Cubic>,<at 0.155 0.2727>,<at 1 0.5133 RCIM_Cubic>]}
        }}))
     "Float (Alpha)"=&FloatFromCurve(FloatCurve=DataInterface() {propertyValues {Curve {
         defaultValue=0
         keys [
          {interpMode=RCIM_Cubic, leaveTangent=18.0304, leaveTangentWeight=0.0371, tangentMode=RCTM_Break, time=0.095, value=1}
          {arriveTangent=-3.2936, interpMode=RCIM_Cubic, leaveTangent=-3.2936, tangentMode=RCTM_User, time=0.2062, value=1}
          <at 1 0 RCIM_Cubic>
         ]
        }}})
    )
    Particles.Corona Size:Vector2f=&Vector2DFromFloat(Value=&Multiply_Float(
      A=18
      B=&FloatFromCurve(FloatCurve=DataInterface() {propertyValues {Curve {
          defaultValue=1
          keys [{interpMode=RCIM_Cubic, tangentMode=RCTM_User, time=0, value=1}, {interpMode=RCIM_Cubic, tangentMode=RCTM_User, time=1, value=0}]
         }}})
     ))
   ) {isSetParameters=true, module=SetVariables_#7}
   /SetVariables(Particles.Blend Glow Color:LinearColor=&MakeLinearColorFromVectorAndFloat(
     "Vector (RGB)"=&MakeVectorFromLinearColorRGB(Color=<rgba 0.8606 0.1953 1 1>)
     "Float (Alpha)"=&FloatFromCurve(FloatCurve=DataInterface() {propertyValues {Curve {
         defaultValue=1
         keys [{interpMode=RCIM_Cubic, tangentMode=RCTM_User, time=0.0857, value=1}, {interpMode=RCIM_Cubic, tangentMode=RCTM_User, time=1, value=0}]
        }}})
    )) {isSetParameters=true, module=SetVariables_#8}
   /SetVariables(
    Particles.ClockHandSize:Vector2f=&Multiply_Vector2D(
     A=$Particles.Initial.ClockHandSize
     B=&MakeVector2D(X=&Waveform("Waveform [1]"=::NewEnumerator0, "[1] Frequency"=3, "[1] Amplitude Min/Max"=<-1 1>), Y=1)
    )
    Particles.ClockPartSize:Vector2f=&Multiply_Vector2D(
     A=$Particles.Initial.SpriteSize
     B=&MakeVector2D(X=&Waveform("Waveform [1]"=::NewEnumerator0, "[1] Frequency"=3, "[1] Amplitude Min/Max"=<-1 1>), Y=1)
    )
   ) {isSetParameters=true, module=SetVariables_#9}
   &MeshRotationForce(Rotation=&Multiply_Vector(
     A=&RandomRangeVector(Minimum=<-1 -1 -1>, Maximum=<1 1 1>)
     B=&VectorFromFloat(Value=&RandomRangeFloat(Minimum=50, Maximum=200))
    ))
   &SolveForcesAndVelocity(Clamp Velocity=false, Limit Acceleration=false)
   &ScaleSpriteSizeBySpeed(
    Max Scale Factor=&MergeHelper_Vector2Switch(0=<1 1>, 1=<1 1>, 2=<1 1>, 3=<1 1>, 4=<1 3>, 5=<1 1>, Index=&SpawnGroup(), Num=::NewEnumerator4)
    Min Scale Factor=&MergeHelper_Vector2Switch(0=<1 1>, 1=<1 1>, 2=<1 1>, 3=<1 1>, 4=<1 1>, 5=<1 1>, Index=&SpawnGroup(), Num=::NewEnumerator4)
    Sample Scale Factor By Curve=true
    Scale Factor Curve=DataInterface() {propertyValues {Curve {keys [
        {arriveTangent=2.865, interpMode=RCIM_Cubic, time=0.1247, value=-0.0006}
        {arriveTangent=4.1522, interpMode=RCIM_Cubic, leaveTangent=4.1522, tangentMode=RCTM_User, time=0.3948, value=0.4893}
        <at 0.9958 0.9974 RCIM_Cubic>
       ]}}}
   )
   &BlendSpace("Position Blend (World <-> Local)":NiagaraFloat=&MakeCustomFloatFromBool(
     Bool=&SetBoolByFloatComparison(A=&VectorLength(Vector=$User.OwnerLocalVelocity), B=$Emitter.WalkForwardSpeed)
     True Float=1
     False Float=&FloatFromCurve(FloatCurve=DataInterface() {propertyValues {Curve {keys [<at 0.6 1>,<at 0.8 0.25>]}}})
     Translation Contribution:NiagaraFloat=&MergeHelper_FloatSwitch(
      0=&MakeCustomFloatFromBool(
       Bool=&SetBoolByFloatComparison(A=&VectorLength(Vector=$User.OwnerLocalVelocity), B=$Emitter.WalkForwardSpeed)
       True Float=1
       False Float=&FloatFromCurve(FloatCurve=DataInterface() {propertyValues {Curve {keys [<at 0 1 RCIM_Cubic>,<at 0.5 0 RCIM_Cubic>]}}})
      )
      1=&MakeCustomFloatFromBool(
       Bool=&SetBoolByFloatComparison(A=&VectorLength(Vector=$User.OwnerLocalVelocity), B=$Emitter.WalkForwardSpeed)
       True Float=1
       False Float=&FloatFromCurve(FloatCurve=DataInterface() {propertyValues {Curve {keys [<at 0 1 RCIM_Cubic>,<at 0.5 0 RCIM_Cubic>]}}})
      )
      2=&MakeCustomFloatFromBool(
       Bool=&SetBoolByFloatComparison(A=&VectorLength(Vector=$User.OwnerLocalVelocity), B=$Emitter.WalkForwardSpeed)
       True Float=1
       False Float=&FloatFromCurve(FloatCurve=DataInterface() {propertyValues {Curve {keys [<at 0 1 RCIM_Cubic>,<at 0.5 0 RCIM_Cubic>]}}})
      )
      3=&MakeCustomFloatFromBool(
       Bool=&SetBoolByFloatComparison(A=&VectorLength(Vector=$User.OwnerLocalVelocity), B=$Emitter.WalkForwardSpeed)
       True Float=1
       False Float=0
      )
      4=&MakeCustomFloatFromBool(
       Bool=&SetBoolByFloatComparison(A=&VectorLength(Vector=$User.OwnerLocalVelocity), B=$Emitter.WalkForwardSpeed)
       True Float=1
       False Float=0
      )
      5=&MakeCustomFloatFromBool(
       Bool=&SetBoolByFloatComparison(A=&VectorLength(Vector=$User.OwnerLocalVelocity), B=$Emitter.WalkForwardSpeed)
       True Float=1
       False Float=&FloatFromCurve(FloatCurve=DataInterface() {propertyValues {Curve {keys [<at 0 1 RCIM_Cubic>,<at 0.5 0 RCIM_Cubic>]}}})
      )
      Index=&SpawnGroup()
      Num=::NewEnumerator4
     )
     Rotation Contribution:NiagaraFloat=&MergeHelper_FloatSwitch(
      0=&MakeCustomFloatFromBool(
       Bool=&SetBoolByFloatComparison(A=&VectorLength(Vector=$User.OwnerLocalVelocity), B=$Emitter.WalkForwardSpeed)
       True Float=1
       False Float=&FloatFromCurve(FloatCurve=DataInterface() {propertyValues {Curve {keys [<at 0 1 RCIM_Cubic>,<at 0.5 0 RCIM_Cubic>]}}})
      )
      1=&MakeCustomFloatFromBool(
       Bool=&SetBoolByFloatComparison(A=&VectorLength(Vector=$User.OwnerLocalVelocity), B=$Emitter.WalkForwardSpeed)
       True Float=1
       False Float=&FloatFromCurve(FloatCurve=DataInterface() {propertyValues {Curve {keys [<at 0 1 RCIM_Cubic>,<at 0.5 0 RCIM_Cubic>]}}})
      )
      2=&MakeCustomFloatFromBool(
       Bool=&SetBoolByFloatComparison(A=&VectorLength(Vector=$User.OwnerLocalVelocity), B=$Emitter.WalkForwardSpeed)
       True Float=1
       False Float=&FloatFromCurve(FloatCurve=DataInterface() {propertyValues {Curve {keys [<at 0 1 RCIM_Cubic>,<at 0.5 0 RCIM_Cubic>]}}})
      )
      3=&MakeCustomFloatFromBool(
       Bool=&SetBoolByFloatComparison(A=&VectorLength(Vector=$User.OwnerLocalVelocity), B=$Emitter.WalkForwardSpeed)
       True Float=1
       False Float=0
      )
      4=&MakeCustomFloatFromBool(
       Bool=&SetBoolByFloatComparison(A=&VectorLength(Vector=$User.OwnerLocalVelocity), B=$Emitter.WalkForwardSpeed)
       True Float=1
       False Float=0
      )
      5=&MakeCustomFloatFromBool(
       Bool=&SetBoolByFloatComparison(A=&VectorLength(Vector=$User.OwnerLocalVelocity), B=$Emitter.WalkForwardSpeed)
       True Float=1
       False Float=&FloatFromCurve(FloatCurve=DataInterface() {propertyValues {Curve {keys [<at 0 1 RCIM_Cubic>,<at 0.5 0 RCIM_Cubic>]}}})
      )
      Index=&SpawnGroup()
      Num=::NewEnumerator4
     )
    )) {enabled=false}
   &SubUVAnimation(
    Direct Frame Index:NiagaraFloat=&RandomRangeFloat(Minimum=0, Maximum=4)
    "Sprite Renderer "=DataInterface() {propertyValues {SpriteRenderer=&self:Main_Burst_Condensed001_0.NiagaraSpriteRendererProperties_6}}
    SubUV Animation Mode:ENiagara_SubUVLookupModeV2=::NewEnumerator5
   )
   &SubUVAnimation(
    Direct Frame Index:NiagaraFloat=&RandomRangeFloat(Minimum=0, Maximum=6)
    "Sprite Renderer "=DataInterface() {propertyValues {SpriteRenderer=&self:Main_Burst_Condensed001_0.NiagaraSpriteRendererProperties_5}}
    SubUV Animation Mode:ENiagara_SubUVLookupModeV2=::NewEnumerator5
   ) {module=SubUVAnimation001}
  ]
  rendererClasses [&NiagaraSpriteRendererProperties, &NiagaraMeshRendererProperties]
  renderers [
   {rendererClass=&NiagaraSpriteRendererProperties}
   {rendererClass=&NiagaraSpriteRendererProperties, rendererIndex=1}
   {rendererClass=&NiagaraSpriteRendererProperties, rendererIndex=2}
   {rendererClass=&NiagaraSpriteRendererProperties, rendererIndex=3}
   {rendererClass=&NiagaraSpriteRendererProperties, rendererIndex=4}
   {rendererClass=&NiagaraSpriteRendererProperties, rendererIndex=5}
   {rendererClass=&NiagaraSpriteRendererProperties, rendererIndex=6}
   {rendererClass=&NiagaraMeshRendererProperties, rendererIndex=7}
  ]
 }
 Main_Star {
  bEnabled=false
  emitterUpdateScript [
   &EmitterState(Life Cycle Mode=::NewEnumerator1, Loop Behavior=::NewEnumerator1, Loop Duration=2, UseLoopDelay=false)
   &SpawnBurst_Instantaneous(Use Spawn Probability=false, UseLoopCountLimit=false)
   &SpawnBurst_Instantaneous(Spawn Group=1, Use Spawn Probability=false, UseLoopCountLimit=false) {module=SpawnBurst_Instantaneous001}
   /SetVariables(Emitter.WalkForwardSpeed=145) {isSetParameters=true}
  ]
  particleSpawnScript [
   /SetVariables(Particles.VisibilityTag=&SpawnGroup()) {isSetParameters=true}
   &InitializeParticle(
    Color=&MaskLinearColor(
     Mask Source=0
     Mask Parameter=?
     Mask=0
     Masked Linear Color=&FNTimeOfDayColor(
      Night Color=<rgba 0.5552 0.6748 1 1>
      Morning Color=<rgba 0.8823 1.0215 1.4 1>
      Day Color=<rgba 1.41 0.8275 4 1>
      Evening Color=<rgba 0.8823 1.0215 1.4 1>
     )
     Passthrough Linear Color=&FNTimeOfDayColor(
      Night Color=<rgba 0.7775 0.2396 2 1>
      Morning Color=<rgba 0.7775 0.2396 2 1>
      Day Color=<rgba 0.7775 0.2396 2 1>
      Evening Color=<rgba 0.7775 0.2396 2 1>
     )
    )
    Lifetime=&MaskFloat(Mask Source=0, Mask Parameter=?, Mask=0, Masked Float=0.7, Passthrough Float=0.25)
    Mesh Scale=<0.025 0.04 0.04>
    Mesh Scale Mode=::NewEnumerator4
    Position Offset=&MaskVector(Mask Source=0, Mask Parameter=?, Mask=0, Masked Vector=<0 0 2>, Passthrough Vector=<0 -30 0>)
    Position Offset Coordinate Space=::Local
    Sprite Rotation Mode=::NewEnumerator2
    Sprite Size Mode=::NewEnumerator3
    Uniform Sprite Size=35
    UsePositionOffset=true
    Write Material Random=true
   )
   &InitialMeshOrientation(Mesh Orientation Mode=::NewEnumerator6, Orientation Quaternion:Quat4f=<0 0 1 1>, Use Rotation Vector=true)
   &ApplyInitialForces()
   &SpriteFacingAndAlignment(
    Alignment Coordinate Space=::Local
    Facing Coordinate Space=::Local
    Sprite Alignment=<0 0 1>
    Sprite Facing=<0 1 0>
    WriteAlignment=true
    WriteFacing=true
   )
   &AddVelocity(Velocity=&MaskVector(Mask Source=0, Mask Parameter=?, Mask=0, Masked Vector=<0 35 0>, Passthrough Vector=<0 0 0>))
   /SetVariables(Particles.Vector:Vector3f=<1 0 1>, Particles.Vector001:Vector3f=<0.5 1 0>) {isSetParameters=true}
   /SetVariables() {isSetParameters=true}
  ]
  particleUpdateScript [
   &ParticleState()
   &ScaleSpriteSize(Uniform Curve Sprite Scale=DataInterface() {propertyValues {Curve {keys [
        {arriveTangent=2.1069, interpMode=RCIM_Cubic, leaveTangent=2.1069, tangentMode=RCTM_User, time=0, value=0.75}
        {arriveTangent=0.3204, interpMode=RCIM_Cubic, leaveTangent=0.3204, time=0.1932, value=0.982}
        <at 0.9963 1.0692 RCIM_Cubic>
       ]}}})
   &ScaleMeshSize(Scale Factor=&Multiply_VectorByFloat(
     Vector=<1 1 1>
     Float=&FloatFromCurve(FloatCurve=DataInterface() {propertyValues {Curve {defaultValue=1, keys [<at 0 0.75 RCIM_Cubic>,<at 1 1.1 RCIM_Cubic>]}}})
    )) {module=ScaleMeshSize001}
   &ScaleColor(
    Scale Alpha=&MaskFloat(
     Mask Source=0
     Mask Parameter=?
     Mask=0
     Masked Float=&FloatFromCurve(
      FloatCurve=DataInterface() {propertyValues {Curve {keys [
          <at 0.2358 1 RCIM_Cubic>
          {arriveTangent=-1.9891, interpMode=RCIM_Cubic, leaveTangent=-1.9891, tangentMode=RCTM_User, time=0.3953, value=0.4222}
          <at 1.0011 0.002 RCIM_Cubic>
         ]}}}
      Scale Curve=&RemapRange(
       Input Value To Remap=&Abs_Float(Float:NiagaraFloat=&DotProduct(
         Vector A=&ReturnCameraVector(Camera Vectors=::NewEnumerator4, Coordinate Space=::Simulation, Origin Position=$Particles.Position)
         Vector B=$Particles.SpriteFacing
        ))
       Input Min=0.6
       Input Max=0.9
       Output Min=0
       Output Max=1
      )
     )
     Passthrough Float=&FloatFromCurve(FloatCurve=DataInterface() {propertyValues {Curve {keys [
          <at -0.0046 1.0113 RCIM_Cubic>
          {arriveTangent=-2.6515, interpMode=RCIM_Cubic, leaveTangent=-2.6515, tangentMode=RCTM_User, time=0.1872, value=0.9143}
          {arriveTangent=-1.392, interpMode=RCIM_Cubic, leaveTangent=-1.392, tangentMode=RCTM_User, time=0.2873, value=0.3784}
          <at 1.0011 0.002 RCIM_Cubic>
         ]}}})
    )
    Scale RGB=&MaskVector(
     Mask Source=0
     Mask Parameter=?
     Mask=0
     Masked Vector=&MakeVectorFromLinearColorRGB(Color=&ColorFromCurve(ColorCurve=DataInterface() {propertyValues {
         AlphaCurve {keys []}
         BlueCurve {keys [
           <at 0.0817 12.5>
           {arriveTangent=-236.5105, interpMode=RCIM_Cubic, leaveTangent=-236.5105, time=0.1288, value=1}
           <at 0.1303 1>
           <at 0.3309 1>
           <at 1 1 RCIM_Cubic>
          ]}
         GreenCurve {keys [
           <at 0.0817 12.5>
           {arriveTangent=-236.5105, interpMode=RCIM_Cubic, leaveTangent=-236.5105, time=0.1288, value=1}
           <at 0.1303 1>
           <at 0.3309 1>
           <at 1 1 RCIM_Cubic>
          ]}
         RedCurve {keys [
           <at 0.0817 12.5>
           {arriveTangent=-236.5105, interpMode=RCIM_Cubic, leaveTangent=-236.5105, time=0.1288, value=1}
           <at 0.1303 1>
           <at 0.3309 1>
           <at 1 1 RCIM_Cubic>
          ]}
        }}))
     Passthrough Vector=<1 1 1>
    )
    ScaleA=true
    ScaleRGB=true
   ) {module=ScaleColor002}
   /SetVariables(Particles.Linear Color=&MakeLinearColorFromVectorAndFloat(
     "Vector (RGB)"=&MakeVectorFromLinearColorRGB(Color=&MultiplyLinearColor(
       Linear Color=$Particles.Initial.Color
       Scale Factor=&MakeVector4_FromLinearColor(LinearColor=&ColorFromCurve(ColorCurve=DataInterface() {propertyValues {
           AlphaCurve {keys [<at 0 1.75>]}
           BlueCurve {keys [<at 0.0576 10>, {arriveTangent=-31.4229, interpMode=RCIM_Cubic, time=0.1288, value=1}, <at 0.1303 1>, <at 0.3309 1>, <at 1 1 RCIM_Cubic>]}
           GreenCurve {keys [<at 0.0576 10>, {arriveTangent=-31.4229, interpMode=RCIM_Cubic, time=0.1288, value=1}, <at 0.1303 1>, <at 0.3309 1>, <at 1 1 RCIM_Cubic>]}
           RedCurve {keys [<at 0.0576 10>, {arriveTangent=-31.4229, interpMode=RCIM_Cubic, time=0.1288, value=1}, <at 0.1303 1>, <at 0.3309 1>, <at 1 1 RCIM_Cubic>]}
          }}))
      ))
     "Float (Alpha)"=&FloatFromCurve(
      FloatCurve=DataInterface() {propertyValues {Curve {keys [
          <at 0.2358 1 RCIM_Cubic>
          {arriveTangent=-1.9891, interpMode=RCIM_Cubic, leaveTangent=-1.9891, tangentMode=RCTM_User, time=0.3953, value=0.4222}
          <at 1.0011 0.002 RCIM_Cubic>
         ]}}}
      Scale Curve=&RemapRange(
       Input Value To Remap=&Abs_Float(Float:NiagaraFloat=&DotProduct(
         Vector A=&ReturnCameraVector(
          Camera Vectors=::NewEnumerator4
          Coordinate Space=::Simulation
          Origin Position=$Particles.Position
          CameraQuery=DataInterface() {propertyValues {bRequireCurrentFrameData=false}}
         )
         Vector B=$Particles.SpriteFacing
        ))
       Input Min=0
       Input Max=0.6
       Output Min=1
       Output Max=0
      )
     )
    )) {isSetParameters=true}
   &Drag(Use Linear Drag=true, Use Rotational Drag=true)
   &SolveForcesAndVelocity(Clamp Velocity=false, Limit Acceleration=false)
   &CameraOffset(Camera Offset Amount=&RemapRange(
     Input Value To Remap=&Abs_Float(Float:NiagaraFloat=&DotProduct(
       Vector A=&VectorFromCurve(
        VectorCurve=DataInterface() {propertyValues {XCurve {keys [<at -1 1>,<at 0.6 0>]}, YCurve {keys [<at -1 1>,<at 0.6 0>]}, ZCurve {keys [<at -1 1>,<at 0.6 0>]}}}
        CurveIndex=&MakeFloatFromVector(
         Vector=&ReturnCameraVector(
          Camera Vectors=::NewEnumerator1
          Coordinate Space=::Simulation
          CameraQuery=DataInterface() {propertyValues {bRequireCurrentFrameData=false}}
         )
         Channel=::NewEnumerator1
        )
        Scale Curve=<1 0 0>
       )
       Vector B=<1 0 0>
      ))
     Input Min=0.2
     Input Max=0.9
     Output Min=0
     Output Max=10
    )) {module=CameraOffset001}
   &BlendSpace("Position Blend (World <-> Local)":NiagaraFloat=&MakeCustomFloatFromBool(
     Bool=&SetBoolByFloatComparison(A=&VectorLength(Vector=$User.OwnerLocalVelocity), B=$Emitter.WalkForwardSpeed)
     True Float=1
     False Float=&FloatFromCurve(FloatCurve=DataInterface() {propertyValues {Curve {keys [<at 0.6 1>,<at 0.8 0.25>]}}})
     Translation Contribution:NiagaraFloat=&MakeCustomFloatFromBool(
      Bool=&SetBoolByFloatComparison(A=&VectorLength(Vector=$User.OwnerLocalVelocity), B=$Emitter.WalkForwardSpeed)
      True Float=1
      False Float=&FloatFromCurve(FloatCurve=DataInterface() {propertyValues {Curve {keys [<at 0 1 RCIM_Cubic>,<at 0.5 0 RCIM_Cubic>]}}})
     )
     Rotation Contribution:NiagaraFloat=&MakeCustomFloatFromBool(
      Bool=&SetBoolByFloatComparison(A=&VectorLength(Vector=$User.OwnerLocalVelocity), B=$Emitter.WalkForwardSpeed)
      True Float=1
      False Float=&FloatFromCurve(FloatCurve=DataInterface() {propertyValues {Curve {keys [<at 0 1 RCIM_Cubic>,<at 0.5 0 RCIM_Cubic>]}}})
     )
    ))
  ]
  rendererClasses [&NiagaraSpriteRendererProperties, &NiagaraMeshRendererProperties]
  renderers [
   {rendererClass=&NiagaraSpriteRendererProperties}
   {rendererClass=&NiagaraSpriteRendererProperties, rendererIndex=1}
   {rendererClass=&NiagaraMeshRendererProperties, rendererIndex=2}
  ]
 }
 Velocity_Particles {
  emitterUpdateScript [
   &EmitterState(Life Cycle Mode=::NewEnumerator1, Loop Behavior=::NewEnumerator1, Loop Duration=1, UseLoopDelay=false)
   &SpawnBurst_Instantaneous(Spawn Count=5, Use Spawn Probability=false, UseLoopCountLimit=false)
   &SpawnBurst_Instantaneous(Spawn Count=7, Spawn Group=1, Use Spawn Probability=false, UseLoopCountLimit=false) {module=SpawnBurst_Instantaneous001}
   /SetVariables(Emitter.WalkForwardSpeed=145) {isSetParameters=true}
  ]
  particleSpawnScript [
   /SetVariables(Particles.VisibilityTag=&SpawnGroup()) {isSetParameters=true}
   &PartitionParticles(
    Number Of Partitions=/Niagara/DynamicInputs/Helpers/MaskInt(Mask Source:NiagaraInt32=0, Mask Parameter:NiagaraInt32=?, Mask:NiagaraInt32=0, Masked Int:NiagaraInt32=0, Passthrough Int:NiagaraInt32=7)
    Particles Per Partition:NiagaraInt32=&MaskInt(Mask Source:NiagaraInt32=0, Mask Parameter:NiagaraInt32=?, Mask:NiagaraInt32=0, Masked Int:NiagaraInt32=0, Passthrough Int:NiagaraInt32=1)
    Partition Mode:ENiagara_PartitionMode_v2=::NewEnumerator1
   )
   &InitializeParticle(
    Color=&MaskLinearColor(
     Mask Source=0
     Mask Parameter=?
     Mask=0
     Masked Linear Color=<rgba 0.3259 0.2695 1 1>
     Passthrough Linear Color=&SelectLinearColorFromArray(
      Array Sampling Mode=::NewEnumerator0
      Direct Array Index=?
      Color Selection Array=DataInterface() {propertyValues {ColorData [<rgba 0.7671 0.7931 3 1>, <rgba 0.6691 0.7409 3 1>, <rgba 0.4115 1.1072 3 1>, <rgba 0.4423 0.3821 3 1>]}}
      Random Seed=0
     )
    )
    Lifetime=&MaskFloat(
     Mask Source=0
     Mask Parameter=?
     Mask=0
     Masked Float=&RandomRangeFloat(Minimum=0.33, Maximum=0.68)
     Passthrough Float=&RandomRangeFloat(Minimum=0.23, Maximum=0.8)
    )
    Mass Max=2
    Mass Mode=::NewEnumerator3
    Sprite Size=/Niagara/DynamicInputs/Helpers/MaskVector2d(
     Mask Source:NiagaraInt32=0
     Mask Parameter:NiagaraInt32=?
     Mask:NiagaraInt32=0
     Masked Vector2D:Vector2f=&RandomRangeVector2D(Minimum=<2 5>, Maximum=<5 15>)
     Passthrough Vector2D:Vector2f=&Vector2DFromFloat(Value=&MergeHelper_FloatSwitch(0=5, 1=15, 2=3, 3=8, 4=4, 5=13, 6=3, Index=$Particles.PartitionParticles.Partition, Num=::NewEnumerator5))
    )
    Sprite Size Mode=::NewEnumerator4
    UsePositionOffset=false
    Write Material Random=true
   )
   &ShapeLocation(
    MaskSpawnGroup=false
    Non Uniform Scale=&MaskVector(Mask Source=0, Mask Parameter=?, Mask=0, Masked Vector=<0 0 0>, Passthrough Vector=<1 1 1>)
    Offset=&MaskVector(Mask Source=0, Mask Parameter=?, Mask=0, Masked Vector=<0 0 0>, Passthrough Vector=<0 5 0>)
    Offset Mode=::NewEnumerator0
    Override Seed=false
    Sphere Radius=&MaskFloat(Mask Source=0, Mask Parameter=?, Mask=0, Masked Float=0, Passthrough Float=15)
    UseOwnerScale=true
   ) {module=ShapeLocation001}
   &AddVelocity(
    Cone Angle=80
    Cone Axis=&MaskVector(Mask Source=0, Mask Parameter=?, Mask=0, Masked Vector=<0 0.7 1>, Passthrough Vector=<0 0 0>)
    Distribution Along Cone Axis=0.2
    Module.UseConeFalloff=false
    Module.UseVelDistribution=true
    Override Seed=false
    Velocity Mode=::NewEnumerator2
    Velocity Speed=&MaskFloat(
     Mask Source=0
     Mask Parameter=?
     Mask=0
     Masked Float=&RandomRangeFloat(Minimum=400, Maximum=700)
     Passthrough Float=0
    )
   )
   &AddVelocity(
    Module.Use Falloff Distance=false
    Override Seed=false
    Velocity Mode=::NewEnumerator1
    Velocity Speed=&MaskFloat(
     Mask Source=0
     Mask Parameter=?
     Mask=0
     Masked Float=0
     Passthrough Float=&RandomRangeFloat(Minimum=75, Maximum=450)
    )
   ) {module=AddVelocity002}
   &AddVelocity(Velocity=&MaskVector(Mask Source=0, Mask Parameter=?, Mask=0, Masked Vector=<0 125 80>, Passthrough Vector=<0 25 0>)) {module=AddVelocity001}
   &ShapeLocation(
    MaskSpawnGroup=false
    Non Uniform Scale=&MaskVector(Mask Source=0, Mask Parameter=?, Mask=0, Masked Vector=<1 1 1>, Passthrough Vector=<0 0 0>)
    Offset=&MaskVector(Mask Source=0, Mask Parameter=?, Mask=0, Masked Vector=<0 5 -7>, Passthrough Vector=<0 0 0>)
    Offset Mode=::NewEnumerator0
    Override Seed=false
    Sphere Radius=&MaskFloat(Mask Source=0, Mask Parameter=?, Mask=0, Masked Float=1, Passthrough Float=0)
    UseOwnerScale=true
   )
  ]
  particleUpdateScript [
   &ParticleState()
   &SpriteRotationRate(Rotation Rate=&MaskFloat(
     Mask Source=0
     Mask Parameter=?
     Mask=0
     Masked Float=0
     Passthrough Float=&UniformAOrBFloat(A=-500, B=500, Evaluation Type=::NewEnumerator0, Recalculate Random Each Loop=true, Override Seed=false)
    ))
   &GravityForce(Gravity=&MaskVector(Mask Source=0, Mask Parameter=?, Mask=0, Masked Vector=<0 0 -980>, Passthrough Vector=<0 0 0>))
   /SetVariables(Particles.Flicker:NiagaraFloat=&Waveform(
     "Waveform [1]"=::NewEnumerator0
     "[1] Frequency"=&RandomRangeFloat(Minimum=6, Maximum=8)
     "[1] Amplitude Min/Max"=<0 1>
    )) {isSetParameters=true}
   &ScaleColor(
    Scale Alpha=&MaskFloat(
     Mask Source=0
     Mask Parameter=?
     Mask=0
     Masked Float=&Multiply_Float(
      A=&FloatFromCurve(FloatCurve=DataInterface() {propertyValues {Curve {keys [
           {arriveTangent=0.2384, interpMode=RCIM_Cubic, leaveTangent=0.2384, tangentMode=RCTM_User, time=-0.0046, value=1.0113}
           {arriveTangent=-2.7114, interpMode=RCIM_Cubic, leaveTangent=-2.7114, tangentMode=RCTM_User, time=1.0011, value=0.002}
          ]}}})
      B=&Lerp_Float(
       A=1
       B=$Particles.Flicker
       Alpha=&FloatFromCurve(FloatCurve=DataInterface() {propertyValues {Curve {
           defaultValue=1
           keys [
            {arriveTangent=0.59, interpMode=RCIM_Cubic, leaveTangent=0.59, tangentMode=RCTM_User, time=0.3951, value=0}
            {interpMode=RCIM_Cubic, tangentMode=RCTM_User, time=0.67, value=1}
           ]
          }}})
      )
     )
     Passthrough Float=1
    )
    Scale RGB=&MaskVector(
     Mask Source=0
     Mask Parameter=?
     Mask=0
     Masked Vector=<1 1 1>
     Passthrough Vector=&MakeVectorFromLinearColorRGB(Color=&ColorFromCurve(ColorCurve=DataInterface() {propertyValues {
         AlphaCurve {keys []}
         BlueCurve {keys [<at 0.0492 0.7781 RCIM_Cubic>,<at 0.1128 0.5631>,<at 0.2082 0.512>,<at 0.3672 1>,<at 0.551 1>,<at 0.7224 0.9754>,<at 0.8496 1>,<at 1 1 RCIM_Cubic>]}
         GreenCurve {keys [<at 0.0492 2.3858 RCIM_Cubic>,<at 0.1128 0.8375>,<at 0.2082 0.468>,<at 0.3672 0.8718>,<at 0.551 1>,<at 0.7224 0.488>,<at 0.8496 0.5031>,<at 1 0.4957 RCIM_Cubic>]}
         RedCurve {keys [<at 0.0492 4 RCIM_Cubic>,<at 0.1128 1.578>,<at 0.2082 1>,<at 0.3672 0.8869>,<at 0.551 1>,<at 0.7224 1>,<at 0.8496 0.8032>,<at 1 0.4981 RCIM_Cubic>]}
        }}))
    )
    ScaleA=true
    ScaleRGB=true
   ) {module=ScaleColor001}
   &Drag(
    Drag=&MaskFloat(
     Mask Source=0
     Mask Parameter=?
     Mask=0
     Masked Float=&RandomRangeFloat(Minimum=2, Maximum=&RandomRangeFloat(Minimum=0, Maximum=1))
     Passthrough Float=1
    )
    Use Linear Drag=true
    Use Rotational Drag=false
   )
   &ScaleSpriteSize(Scale Factor=&MaskVector2d(
     Mask Source:NiagaraInt32=0
     Mask Parameter:NiagaraInt32=?
     Mask:NiagaraInt32=0
     Masked Vector2D:Vector2f=<1 1>
     Passthrough Vector2D:Vector2f=&Vector2DFromFloat(Value=&FloatFromCurve(FloatCurve=DataInterface() {propertyValues {Curve {keys [
           <at -0.0046 1.0113 RCIM_Cubic>
           {arriveTangent=-1.0036, interpMode=RCIM_Cubic, leaveTangent=-1.0036, time=0.7001, value=0.872}
           <at 1.0011 0.002>
          ]}}}))
    ))
   &ScaleColor(
    Scale Alpha=&MaskFloat(
     Mask Source=0
     Mask Parameter=?
     Mask=0
     Masked Float=&FloatFromCurve(FloatCurve=DataInterface() {propertyValues {Curve {keys [
          <at -0.0046 1.0113 RCIM_Cubic>
          {arriveTangent=-1.0036, interpMode=RCIM_Cubic, leaveTangent=-1.0036, time=0.5034, value=0.1741}
          <at 1.0011 0.002 RCIM_Cubic>
         ]}}})
     Passthrough Float=&FloatFromCurve(FloatCurve=DataInterface() {propertyValues {Curve {keys [
          <at 0.239 1 RCIM_Cubic>
          {arriveTangent=-2.1255, interpMode=RCIM_Cubic, leaveTangent=-2.1255, tangentMode=RCTM_User, time=0.4753, value=0.4021}
          <at 1.0011 0.002 RCIM_Cubic>
         ]}}})
    )
    ScaleA=true
    ScaleRGB=true
   )
   &ScaleVelocity(Velocity Scale=&Multiply_VectorByFloat(
     Vector=<1 1 1>
     Float=&MaskFloat(
      Mask Source=0
      Mask Parameter=?
      Mask=0
      Masked Float=&FloatFromCurve(FloatCurve=DataInterface() {propertyValues {Curve {keys [<at 0.0043 0.9247>,<at 0.1707 0.6944>,<at 0.2122 0.1301>,<at 1 0>]}}})
      Passthrough Float=&FloatFromCurve(FloatCurve=DataInterface() {propertyValues {Curve {keys [<at 0.0013 1.2447>,<at 0.0207 1.0581>,<at 0.1905 1.0018>,<at 0.234 0.5434>,<at 1 0>]}}})
     )
    ))
   &SolveForcesAndVelocity(Clamp Velocity=false, Limit Acceleration=false)
   &ScaleSpriteSizeBySpeed(
    Max Scale Factor=&MaskVector2d(
     Mask Source:NiagaraInt32=0
     Mask Parameter:NiagaraInt32=?
     Mask:NiagaraInt32=0
     Masked Vector2D:Vector2f=<1.2 7>
     Passthrough Vector2D:Vector2f=<1 1>
    )
    Min Scale Factor=&MaskVector2d(
     Mask Source:NiagaraInt32=0
     Mask Parameter:NiagaraInt32=?
     Mask:NiagaraInt32=0
     Masked Vector2D:Vector2f=<0.75 0.7>
     Passthrough Vector2D:Vector2f=<1 1>
    )
    Sample Scale Factor By Curve=true
    Scale Factor Curve=DataInterface() {propertyValues {Curve {keys [
        {arriveTangent=2.865, interpMode=RCIM_Cubic, time=0.1247, value=-0.0006}
        {arriveTangent=4.1522, interpMode=RCIM_Cubic, leaveTangent=4.1522, tangentMode=RCTM_User, time=0.3948, value=0.4893}
        <at 0.9958 0.9974 RCIM_Cubic>
       ]}}}
    Velocity Threshold=1500
   )
   /SetVariables(
    Particles.Linear Color=&MakeLinearColorFromVectorAndFloat(
     "Vector (RGB)"=&MakeVectorFromLinearColorRGB(Color=&FNTimeOfDayColor(
       Night Color=&ColorFromCurve(ColorCurve=DataInterface() {propertyValues {
          AlphaCurve {keys []}
          BlueCurve {keys [
            <at 0.425 1.5 RCIM_Cubic>
            {arriveTangent=-2.2987, interpMode=RCIM_Cubic, leaveTangent=-2.2987, time=0.6763, value=0.7448}
            <at 0.9782 0.2281>
           ]}
          GreenCurve {keys [
            <at 0.425 0.3141 RCIM_Cubic>
            {arriveTangent=-0.5644, interpMode=RCIM_Cubic, leaveTangent=-0.5644, time=0.6763, value=0.1632}
            <at 0.9782 0.0018>
           ]}
          RedCurve {keys [
            <at 0.425 1.5 RCIM_Cubic>
            {arriveTangent=-2.3587, interpMode=RCIM_Cubic, leaveTangent=-2.3587, time=0.6763, value=0.7058}
            <at 0.9782 0.195>
           ]}
         }})
       Morning Color=&ColorFromCurve(ColorCurve=DataInterface() {propertyValues {
          AlphaCurve {keys []}
          BlueCurve {keys [
            <at 0.425 1.5 RCIM_Cubic>
            {arriveTangent=-2.2987, interpMode=RCIM_Cubic, leaveTangent=-2.2987, time=0.6763, value=0.7448}
            <at 0.9782 0.2281>
           ]}
          GreenCurve {keys [
            <at 0.425 0.6578 RCIM_Cubic>
            {arriveTangent=-1.1857, interpMode=RCIM_Cubic, leaveTangent=-1.1857, time=0.6763, value=0.1632}
            <at 0.9782 0.0018>
           ]}
          RedCurve {keys [
            <at 0.425 1.5 RCIM_Cubic>
            {arriveTangent=-2.3587, interpMode=RCIM_Cubic, leaveTangent=-2.3587, time=0.6763, value=0.7058}
            <at 0.9782 0.195>
           ]}
         }})
       Day Color=&ColorFromCurve(ColorCurve=DataInterface() {propertyValues {
          AlphaCurve {keys []}
          BlueCurve {keys [<at 0.425 1.5 RCIM_Cubic>,<at 0.6763 0.7448 RCIM_Cubic>,<at 0.9782 0.2281>]}
          GreenCurve {keys [<at 0.425 0.6931 RCIM_Cubic>,<at 0.6763 0.1632 RCIM_Cubic>,<at 0.9782 0.0018>]}
          RedCurve {keys [<at 0.425 1.5 RCIM_Cubic>,<at 0.6763 0.7058 RCIM_Cubic>,<at 0.9782 0.195>]}
         }})
       Evening Color=&ColorFromCurve(ColorCurve=DataInterface() {propertyValues {
          AlphaCurve {keys []}
          BlueCurve {keys [
            <at 0.425 1.5 RCIM_Cubic>
            {arriveTangent=-2.2987, interpMode=RCIM_Cubic, leaveTangent=-2.2987, time=0.6763, value=0.7448}
            <at 0.9782 0.2281>
           ]}
          GreenCurve {keys [
            <at 0.425 0.6578 RCIM_Cubic>
            {arriveTangent=-1.1857, interpMode=RCIM_Cubic, leaveTangent=-1.1857, time=0.6763, value=0.1632}
            <at 0.9782 0.0018>
           ]}
          RedCurve {keys [
            <at 0.425 1.5 RCIM_Cubic>
            {arriveTangent=-2.3587, interpMode=RCIM_Cubic, leaveTangent=-2.3587, time=0.6763, value=0.7058}
            <at 0.9782 0.195>
           ]}
         }})
      ))
     "Float (Alpha)"=&Multiply_Float(
      A=1
      B=&Multiply_Float(
       A=&FloatFromCurve(FloatCurve=DataInterface() {propertyValues {Curve {keys [
            {arriveTangent=0.2384, interpMode=RCIM_Cubic, leaveTangent=0.2384, tangentMode=RCTM_User, time=-0.0046, value=1.0113}
            {arriveTangent=-2.7114, interpMode=RCIM_Cubic, leaveTangent=-2.7114, tangentMode=RCTM_User, time=1.0011, value=0.002}
           ]}}})
       B=&Lerp_Float(
        A=1
        B=$Particles.Flicker
        Alpha=&FloatFromCurve(FloatCurve=DataInterface() {propertyValues {Curve {
            defaultValue=1
            keys [
             {arriveTangent=0.59, interpMode=RCIM_Cubic, leaveTangent=0.59, tangentMode=RCTM_User, time=0.3951, value=0}
             {interpMode=RCIM_Cubic, tangentMode=RCTM_User, time=0.67, value=1}
            ]
           }}})
       )
      )
     )
    )
    Particles.Vector 2D:Vector2f=&Multiply_Vector2D_ByFloat(Float=&RandomRangeFloat(Minimum=1, Maximum=1.5), Vector2D=$Particles.SpriteSize)
   ) {isSetParameters=true}
   /SetVariables(Particles.Corona size:Vector2f=&Multiply_Vector2D_ByFloat(Float=&RandomRangeFloat(Minimum=0.25, Maximum=0.35), Vector2D=$Particles.SpriteSize)) {isSetParameters=true}
   /SetVariables(Particles.Stars size:Vector2f=&Multiply_Vector2D_ByFloat(
     Float=&SelectFloatFromArray(
      Array Sampling Mode=::NewEnumerator0
      Direct Array Index=?
      Float Selection Array=DataInterface() {propertyValues {FloatData [0.5,1,2]}}
      Random Seed=0
     )
     Vector2D=$Particles.SpriteSize
    )) {isSetParameters=true}
   &BlendSpace("Position Blend (World <-> Local)":NiagaraFloat=&MakeCustomFloatFromBool(
     Bool=&SetBoolByFloatComparison(A=&VectorLength(Vector=$User.OwnerLocalVelocity), B=$Emitter.WalkForwardSpeed)
     True Float=1
     False Float=&FloatFromCurve(FloatCurve=DataInterface() {propertyValues {Curve {keys [<at 0.6 1>,<at 0.8 0.25>]}}})
     Translation Contribution:NiagaraFloat=&MaskFloat(
      Mask Source=0
      Mask Parameter=?
      Mask=0
      Masked Float=&MakeCustomFloatFromBool(
       Bool=&SetBoolByFloatComparison(A=&VectorLength(Vector=$User.OwnerLocalVelocity), B=$Emitter.WalkForwardSpeed)
       True Float=1
       False Float=&FloatFromCurve(FloatCurve=DataInterface() {propertyValues {Curve {keys [<at 0 1 RCIM_Cubic>,<at 0.5 0 RCIM_Cubic>]}}})
      )
      Passthrough Float=&MakeCustomFloatFromBool(
       Bool=&SetBoolByFloatComparison(A=&VectorLength(Vector=$User.OwnerLocalVelocity), B=$Emitter.WalkForwardSpeed)
       True Float=1
       False Float=0
      )
     )
     Rotation Contribution:NiagaraFloat=&MaskFloat(
      Mask Source=0
      Mask Parameter=?
      Mask=0
      Masked Float=&MakeCustomFloatFromBool(
       Bool=&SetBoolByFloatComparison(A=&VectorLength(Vector=$User.OwnerLocalVelocity), B=$Emitter.WalkForwardSpeed)
       True Float=1
       False Float=&FloatFromCurve(FloatCurve=DataInterface() {propertyValues {Curve {keys [<at 0 1 RCIM_Cubic>,<at 0.5 0 RCIM_Cubic>]}}})
      )
      Passthrough Float=&MakeCustomFloatFromBool(
       Bool=&SetBoolByFloatComparison(A=&VectorLength(Vector=$User.OwnerLocalVelocity), B=$Emitter.WalkForwardSpeed)
       True Float=1
       False Float=0
      )
     )
    ))
  ]
  rendererClasses [&NiagaraSpriteRendererProperties]
  renderers [
   {rendererClass=&NiagaraSpriteRendererProperties}
   {rendererClass=&NiagaraSpriteRendererProperties, rendererIndex=1}
  ]
 }
}
summary {
 systemName=NS_Emote_IndoorLaceBeat_Kick
 emitters [
  {
   bEnabled=false
   emitterName=Main_Star
   rendererClasses [&NiagaraSpriteRendererProperties, &NiagaraMeshRendererProperties]
  }
  {
   emitterName=Main_Burst_Condensed001
   simTarget=GPUComputeSim
   rendererClasses [&NiagaraSpriteRendererProperties, &NiagaraMeshRendererProperties]
  }
  {emitterName=Velocity_Particles, rendererClasses [&NiagaraSpriteRendererProperties]}
  {emitterName=Extra_Sparkles, rendererClasses [&NiagaraSpriteRendererProperties]}
  {
   emitterName=Main_Burst_Condensed
   rendererClasses [&NiagaraSpriteRendererProperties, &NiagaraMeshRendererProperties]
  }
 ]
 userVariables [
  User.bIsFrontEnd {type=NiagaraBool}
  User.bIsFrontEndPreview {type=NiagaraBool}
  User.OwnerLocalVelocity {type=Vector3f, description="", defaultValue {struct=Vector3f, value=<0 0 0>}}
 ]
}
