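# Niagara system for the IndoorLaceBeat emote that spawns continuous sprite sparkles around five gear positions.
# A single emitter partitions particles five ways, randomly redirecting some from the smallest gear, then uses switch lookups to place each in an offset, oriented cylinder volume.
# One random scale value drives stepped lifetime and size curves, each with a multiplier.
# Alpha follows a fade curve multiplied by an age-driven waveform for twinkle.
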NiagaraSystem /BRCosmetics/Animation/Game/MainPlayer/Emotes/IndoorLaceBeat/FX/NS_IndoorLaceBeat_GearSparkles
dependencies {usedDataInterfaces [&NiagaraDataInterfaceCurve], usedDynamicInputs [], usedModules [&SystemState]}
emitters {Minimal {
  emitterUpdateScript [
   &EmitterState()
   &SpawnRate(SpawnRate=70, Use Spawn Probability=false)
  ]
  particleSpawnScript [
   /SetVariables(Particles.Scale:NiagaraFloat=&RandomRangeFloat(Minimum=0, Maximum=1)) {isSetParameters=true}
   &InitializeParticle(
    Color Mode=::NewEnumerator0
    Lifetime=&FloatFromCurve(
     FloatCurve=DataInterface() {propertyValues {Curve {keys [<at 0 1 RCIM_Constant>,<at 0.4563 1.4 RCIM_Constant>,<at 0.857 1.1 RCIM_Constant>]}}}
     CurveIndex=$Particles.Scale
     Scale Curve=&RandomRangeFloat(Minimum=0.6, Maximum=1.1)
    )
    Sprite Size Mode=::NewEnumerator3
    Uniform Sprite Size=&FloatFromCurve(
     FloatCurve=DataInterface() {propertyValues {Curve {keys [<at 0 0.3 RCIM_Constant>,<at 0.4406 0.5 RCIM_Constant>,<at 0.8099 1.5 RCIM_Constant>]}}}
     CurveIndex=$Particles.Scale
     Scale Curve=10
    )
    UsePositionOffset=false
    Write Material Random=true
   )
   &PartitionParticles(Number Of Partitions=5)
   /SetVariables(Particles.PartitionParticles.Partition:NiagaraInt32=&MakeCustomIntFromBool(
     Bool=&SetBoolByFloatComparison(
      A=&MakeFloatFromInt(Int=$Particles.PartitionParticles.Partition)
      B=4
      Comparison Type=::NewEnumerator4
     )
     True INT=&MakeCustomIntFromBool(
      Bool=&SetBoolByFloatComparison(A=&RandomRangeFloat(Minimum=0, Maximum=1), B=0.1)
      True INT=&RandomRangeInt(Minimum=0, Maximum=3)
      False INT=4
     )
     False INT=$Particles.PartitionParticles.Partition
    )) {isSetParameters=true}
   /SetVariables(
    Particles.GearAxis:Vector3f=&MergeHelper_VectorSwitch(0=<1 0.45 0>, 1=<1 0.4 0>, 2=<0 1 0>, 3=<0.4 -0.22 0>, 4=<1 0.45 0>, Index=$Particles.PartitionParticles.Partition, Num=::NewEnumerator3)
    Particles.GearHeight:NiagaraFloat=&MergeHelper_FloatSwitch(0=10, 1=10, 2=15, 3=10, 4=10, Index=$Particles.PartitionParticles.Partition, Num=::NewEnumerator3)
    Particles.GearPosition:Vector3f=&MergeHelper_VectorSwitch(
     0=<-80 -50 30>
     1=<-53 -76 80>
     2=<0 -111 65>
     3=<60 -62 40>
     4=<-65 -76 12>
     Index=$Particles.PartitionParticles.Partition
     Num=::NewEnumerator3
    )
    Particles.GearRadius:NiagaraFloat=&MergeHelper_FloatSwitch(0=25, 1=25, 2=44, 3=45, 4=10, Index=$Particles.PartitionParticles.Partition, Num=::NewEnumerator3)
   ) {isSetParameters=true}
   &ShapeLocation(
    Cylinder Height=$Particles.GearHeight
    Cylinder Radius=$Particles.GearRadius
    MaskSpawnGroup=false
    Offset=$Particles.GearPosition
    Offset Mode=::NewEnumerator0
    Override Seed=false
    Rotation Axis=$Particles.GearAxis
    Rotation Mode=::NewEnumerator1
    Shape Primitive=::NewEnumerator1
    Surface Only=false
    UseOwnerScale=true
   )
   &CameraOffset(Camera Offset Amount=20)
  ]
  particleUpdateScript [
   &ParticleState()
   &ScaleColor(
    Scale Alpha=&FloatFromCurve(
     FloatCurve=DataInterface() {propertyValues {Curve {
        defaultValue=1
        keys [
         {interpMode=RCIM_Cubic, tangentMode=RCTM_User, time=0, value=0}
         {interpMode=RCIM_Cubic, tangentMode=RCTM_User, time=0.2269, value=1}
         {arriveTangent=-0.1174, interpMode=RCIM_Cubic, leaveTangent=-0.1174, tangentMode=RCTM_User, time=0.3787, value=1}
         {tangentMode=RCTM_User, time=1, value=0}
        ]
       }}}
     Scale Curve=&Waveform(Global Phase=$Particles.Age, "Waveform [1]"=::NewEnumerator0, "[1] Frequency"=5, "[1] Amplitude Min/Max"=<0 1>, "[1] Bias"=0.3)
    )
    ScaleA=true
    ScaleRGB=true
   )
  ]
  rendererClasses [&NiagaraSpriteRendererProperties]
  renderers [{rendererClass=&NiagaraSpriteRendererProperties}]
 }}
summary {
 systemName=NS_IndoorLaceBeat_GearSparkles
 emitters [{emitterName=Minimal, rendererClasses [&NiagaraSpriteRendererProperties]}]
 userVariables []
}
