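# Niagara system for the IndoorLaceBeat player emote: a glowing, rotating ground clock with lights, gear glows and sparkles.
# Clock_and_Lights draws the clock with decals; FE_Clock is a sprite counterpart, apparently for front end.
# Gear sparkles fill per-partition oriented cylinders, with glows driven by System.Pulsate.
# Flickering ground sparkles follow traced slopes, while hand trails sample two skeletal bone/socket points, inheriting velocity, with user-controlled rate and length.
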NiagaraSystem /BRCosmetics/Animation/Game/MainPlayer/Emotes/IndoorLaceBeat/FX/NS_IndoorLaceBeat_Gears
dependencies {
 usedDataInterfaces [
  &NiagaraDataInterfaceCurve
  &NiagaraDataInterfaceArrayFloat3
  &NiagaraDataInterfaceCollisionQuery
  &NiagaraDataInterfaceDebugDraw
  &NiagaraDataInterfaceSkeletalMesh
 ]
 usedDynamicInputs [&Boolean_OrOperation]
 usedModules [
  &SystemState
  &self:SystemSpawnScript.SystemScriptSource.SystemScriptGraph.NiagaraNodeAssignment_0.SetVariables_#1
  &self:SystemSpawnScript.SystemScriptSource.SystemScriptGraph.NiagaraNodeAssignment_1.SetVariables_#2
  &self:SystemSpawnScript.SystemScriptSource.SystemScriptGraph.NiagaraNodeAssignment_3.SetVariables_#3
 ]
}
emitters {
 Clock_and_Lights {
  emitterUpdateScript [
   &EmitterState(Life Cycle Mode=::NewEnumerator1, Loop Behavior=::NewEnumerator1, Loop Duration=50, UseLoopDelay=false)
   &SpawnBurst_Instantaneous(Use Spawn Probability=false, UseLoopCountLimit=false) {module=SpawnBurst_Instantaneous002}
   /SetVariables(Emitter.DecalRadius:NiagaraFloat=&FloatFromCurve(
     FloatCurve=DataInterface() {propertyValues {Curve {keys [{arriveTangent=1.466, interpMode=RCIM_Cubic, leaveTangent=1.466, tangentMode=RCTM_User, time=0, value=0}, <at 0.6 0.5 RCIM_Cubic>]}}}
     CurveIndex=$Emitter.Age
    )) {isSetParameters=true}
   &SpawnBurst_Instantaneous(Spawn Count=3, Spawn Group=1, Use Spawn Probability=false, UseLoopCountLimit=false) {module=SpawnBurst_Instantaneous001}
  ]
  particleSpawnScript [
   &PartitionParticles(Number Of Partitions=3)
   /SetVariables(Particles.PartitionParticles.Partition=2) {enabled=false, isSetParameters=true}
   &InitializeParticle(
    Color=&MaskLinearColor(
     Mask Source=1
     Mask Parameter=&SpawnGroup()
     Mask=0
     Masked Linear Color=<rgba 0.2792 0.3327 0.6979 0.6>
     Passthrough Linear Color=<rgba 2.6597 1.8061 10 1>
    )
    Lifetime=0.1
    Position Offset=<0 -40 0>
    Sprite Size Mode=::NewEnumerator3
    UsePositionOffset=true
    Write Material Random=true
   )
   /SetVariables(Particles.LightPosition:NiagaraPosition=&AddVectorToPosition(
     Position=/Niagara/DynamicInputs/Helpers/SimulationPosition()
     Vector=&SelectVectorFromArray(
      Direct Array Index=$Particles.PartitionParticles.Partition
      Vector Selection Array=DataInterface() {propertyValues {FloatData [<0 -40 10>, <0 -80 20>, <0 0 40>]}}
     )
    )) {isSetParameters=true}
   /SetVariables(Particles.VisibilityTag=&SpawnGroup()) {isSetParameters=true}
   &Light_Attributes(
    Diffuse Scale=1
    Falloff Exponent=1
    Light Diffuse Scale:NiagaraFloat=&MergeHelper_FloatSwitch(0=3, 1=5, 2=1, 3=?, Index=$Particles.PartitionParticles.Partition, Num=::NewEnumerator1)
    Light Exponent:NiagaraFloat=&FN_Select_Float_from_Time_Of_Day(Morning=1, Day=1, Evening=1, Night=1)
    Light Radius:NiagaraFloat=&MergeHelper_FloatSwitch(0=80, 1=120, 2=110, 3=?, Index=$Particles.PartitionParticles.Partition, Num=::NewEnumerator1)
    Light Specular Scale:NiagaraFloat=&MergeHelper_FloatSwitch(0=3, 1=2, 2=1, 3=?, Index=$Particles.PartitionParticles.Partition, Num=::NewEnumerator1)
    Radius=1
    Specular Scale=1
   )
   /SetVariables(
    Particles.GlowDecalColor:LinearColor=<rgba 0.6957 0.758 2 0.1>
    Particles.GlowDecalPos:NiagaraPosition=&AddVectorToPosition(Position=$Particles.Position, Vector=<-1 -2 0>)
    Particles.GlowDecalSize:Vector3f=<60 160 160>
   ) {isSetParameters=true}
  ]
  particleUpdateScript [
   &ParticleState(Kill Particles When Lifetime Has Elapsed=false)
   /Niagara/Modules/Decal/Decal_Attributes(
    Decal Facing Mode=1
    Decal Rotation Mode=2
    Decal Size:Vector3f=&Multiply_VectorByFloat(Vector=<60 250 250>, Float=1)
    Facing Coordinate Space:ENiagaraCoordinateSpace=::World
    Orientation=1
    Rotation Angle:NiagaraFloat=&Add_Float(
     A=&Multiply_Float(A=$Particles.Age, B=15)
     B=&FloatFromCurve(
      FloatCurve=DataInterface() {propertyValues {Curve {keys [{arriveTangent=-2.4903, interpMode=RCIM_Cubic, leaveTangent=-2.4903, tangentMode=RCTM_User, time=0, value=1}, <at 1 0 RCIM_Cubic>]}}}
      CurveIndex=$Particles.Age
      Scale Curve=-25
     )
    )
    Set Alignment=false
    Size=1
    Visible=1
   ) {module=Decal_Attributes001}
   &ScaleColor(
    Scale Alpha=&FloatFromCurve(
     FloatCurve=DataInterface() {propertyValues {Curve {keys [{arriveTangent=1.7855, interpMode=RCIM_Cubic, leaveTangent=1.7855, tangentMode=RCTM_User, time=0, value=0}, <at 1 1 RCIM_Cubic>]}}}
     CurveIndex=$Particles.Age
    )
    ScaleA=true
    ScaleRGB=true
   )
   /SetVariables(Particles.GlowDecalColor:LinearColor=&MultiplyLinearColor(
     Linear Color=$Particles.Initial.GlowDecalColor
     Scale Factor=&MakeVector4FromVectorAndFloat(
      "Vector (RGB)"=&VectorFromFloat(Value=&MakeFloatFromLinearColor(
        LinearColor=&MultiplyLinearColor(
         Linear Color=$Particles.Initial.GlowDecalColor
         Scale Factor=&MakeVector4_FromFloat(Float=$System.Pulsate)
        )
        Channel:ENiagaraFloatFromLinearColorOptions=::NewEnumerator0
       ))
      "Float (Alpha)"=&FloatFromCurve(
       FloatCurve=DataInterface() {propertyValues {Curve {keys [{arriveTangent=1.7855, interpMode=RCIM_Cubic, leaveTangent=1.7855, tangentMode=RCTM_User, time=0, value=0}, <at 1 1 RCIM_Cubic>]}}}
       CurveIndex=$Particles.Age
      )
     )
    )) {isSetParameters=true}
  ]
  rendererClasses [&NiagaraDecalRendererProperties, &NiagaraLightRendererProperties]
  renderers [
   {rendererClass=&NiagaraDecalRendererProperties, rendererIndex=0}
   {rendererClass=&NiagaraLightRendererProperties, rendererIndex=1}
   {rendererClass=&NiagaraDecalRendererProperties, rendererIndex=2}
  ]
 }
 FE_Clock {
  emitterUpdateScript [
   &EmitterState(Life Cycle Mode=::NewEnumerator1, Loop Behavior=::NewEnumerator1, Loop Duration=50, UseLoopDelay=false)
   &SpawnBurst_Instantaneous(Use Spawn Probability=false, UseLoopCountLimit=false)
   &SpawnBurst_Instantaneous(Spawn Count=3, Spawn Group=1, Use Spawn Probability=false, UseLoopCountLimit=false) {module=SpawnBurst_Instantaneous001}
  ]
  particleSpawnScript [
   &PartitionParticles(Number Of Partitions=3)
   &InitializeParticle(
    Color=&MaskLinearColor(
     Mask Source=1
     Mask Parameter=&SpawnGroup()
     Mask=0
     Masked Linear Color=<rgba 0.3985 0.479 0.8646 0.55>
     Passthrough Linear Color=<rgba 2.6597 1.8061 10 1>
    )
    Lifetime=0.1
    Position Offset=<0 -40 2.1>
    Sprite Size Mode=::NewEnumerator3
    Uniform Sprite Size=250
    UsePositionOffset=true
    Write Material Random=true
   )
   /SetVariables(Particles.VisibilityTag=&SpawnGroup()) {isSetParameters=true}
   /SetVariables(Particles.LightPosition:NiagaraPosition=&AddVectorToPosition(
     Position=&SimulationPosition()
     Vector=&SelectVectorFromArray(
      Direct Array Index=$Particles.PartitionParticles.Partition
      Vector Selection Array=DataInterface() {propertyValues {FloatData [<0 -40 10>, <0 -60 20>, <0 0 40>]}}
     )
    )) {isSetParameters=true}
   &SpriteFacingAndAlignment(Sprite Facing=<0 0 1>, WriteAlignment=false, WriteFacing=true)
   &Light_Attributes(
    Diffuse Scale=1
    Falloff Exponent=1
    Light Diffuse Scale:NiagaraFloat=&MergeHelper_FloatSwitch(0=3, 1=10, 2=3, 3=?, Index=$Particles.PartitionParticles.Partition, Num=::NewEnumerator1)
    Light Exponent:NiagaraFloat=&FN_Select_Float_from_Time_Of_Day(Morning=1, Day=1, Evening=1, Night=1)
    Light Radius:NiagaraFloat=&MergeHelper_FloatSwitch(0=80, 1=120, 2=110, 3=?, Index=$Particles.PartitionParticles.Partition, Num=::NewEnumerator1)
    Light Specular Scale:NiagaraFloat=&MergeHelper_FloatSwitch(0=0, 1=0, 2=0, 3=?, Index=$Particles.PartitionParticles.Partition, Num=::NewEnumerator1)
    Radius=1
    Specular Scale=1
   )
   /SetVariables(
    Particles.GlowDecalColor:LinearColor=<rgba 0.6957 0.758 2 0.1>
    Particles.GlowDecalPos:NiagaraPosition=&AddVectorToPosition(Position=$Particles.Position, Vector=<-1 -2 0>)
    Particles.GlowSpriteSize:Vector2f=<170 170>
   ) {isSetParameters=true}
  ]
  particleUpdateScript [
   &ParticleState(Kill Particles When Lifetime Has Elapsed=false)
   /SetVariables(Particles.SpriteRotation:NiagaraFloat=&Add_Float(
     A=&Multiply_Float(A=$Particles.Age, B=15)
     B=&FloatFromCurve(
      FloatCurve=DataInterface() {propertyValues {Curve {keys [{arriveTangent=-2.4903, interpMode=RCIM_Cubic, leaveTangent=-2.4903, tangentMode=RCTM_User, time=0, value=1}, <at 1 0 RCIM_Cubic>]}}}
      CurveIndex=$Particles.Age
      Scale Curve=-25
     )
    )) {isSetParameters=true}
   /SetVariables(Particles.GlowDecalColor:LinearColor=&MultiplyLinearColor(
     Linear Color=$Particles.Initial.GlowDecalColor
     Scale Factor=&MakeVector4FromVectorAndFloat(
      "Vector (RGB)"=&VectorFromFloat(Value=&MakeFloatFromLinearColor(
        LinearColor=&MultiplyLinearColor(
         Linear Color=$Particles.Initial.GlowDecalColor
         Scale Factor=&MakeVector4_FromFloat(Float=$System.Pulsate)
        )
        Channel:ENiagaraFloatFromLinearColorOptions=::NewEnumerator0
       ))
      "Float (Alpha)"=&FloatFromCurve(
       FloatCurve=DataInterface() {propertyValues {Curve {keys [{arriveTangent=1.7855, interpMode=RCIM_Cubic, leaveTangent=1.7855, tangentMode=RCTM_User, time=0, value=0}, <at 1 1 RCIM_Cubic>]}}}
       CurveIndex=$Particles.Age
      )
     )
    )) {isSetParameters=true}
  ]
  rendererClasses [&NiagaraSpriteRendererProperties, &NiagaraLightRendererProperties]
  renderers [
   {rendererClass=&NiagaraSpriteRendererProperties}
   {rendererClass=&NiagaraLightRendererProperties, rendererIndex=1}
   {rendererClass=&NiagaraSpriteRendererProperties, rendererIndex=2}
  ]
 }
 Gear_Sparkles {
  emitterUpdateScript [
   &EmitterState(Life Cycle Mode=::NewEnumerator1, Loop Duration=5000, UseLoopDelay=false)
   &SpawnBurst_Instantaneous(Spawn Count=8, Spawn Group=1, Use Spawn Probability=false, UseLoopCountLimit=false)
   &SpawnRate(SpawnRate=35, Use Spawn Probability=false)
  ]
  particleSpawnScript [
   /SetVariables(Particles.VisibilityTag=$Emitter.SpawnGroup) {isSetParameters=true}
   /SetVariables(Particles.Scale:NiagaraFloat=&RandomRangeFloat(Minimum=0, Maximum=1)) {isSetParameters=true, module=SetVariables_#4}
   &InitializeParticle(
    Color Mode=::NewEnumerator0
    Lifetime=&FloatFromCurve(
     FloatCurve=DataInterface() {propertyValues {Curve {keys [<at 0 1 RCIM_Constant>,<at 0.4563 1.4 RCIM_Constant>,<at 0.88 1.1 RCIM_Constant>]}}}
     CurveIndex=$Particles.Scale
     Scale Curve=&RandomRangeFloat(Minimum=0.6, Maximum=1.1)
    )
    Sprite Size Mode=::NewEnumerator3
    Uniform Sprite Size=&FloatFromCurve(
     FloatCurve=DataInterface() {propertyValues {Curve {keys [<at 0 0.3 RCIM_Constant>,<at 0.4406 0.5 RCIM_Constant>,<at 0.88 1.5 RCIM_Constant>]}}}
     CurveIndex=$Particles.Scale
     Scale Curve=15
    )
    UsePositionOffset=false
    Write Material Random=true
   )
   &PartitionParticles(Number Of Partitions=5)
   /SetVariables(Particles.PartitionParticles.Partition:NiagaraInt32=&MakeCustomIntFromBool(
     Bool=&SetBoolByFloatComparison(
      A=&MakeFloatFromInt(Int=$Particles.PartitionParticles.Partition)
      B=4
      Comparison Type=::NewEnumerator4
     )
     True INT=&MakeCustomIntFromBool(
      Bool=&SetBoolByFloatComparison(A=&RandomRangeFloat(Minimum=0, Maximum=1), B=0.1)
      True INT=&RandomRangeInt(Minimum=0, Maximum=3)
      False INT=4
     )
     False INT=$Particles.PartitionParticles.Partition
    )) {isSetParameters=true}
   /SetVariables() {isSetParameters=true}
   /SetVariables(
    Particles.GearAxis:Vector3f=&MergeHelper_VectorSwitch(
     0=<1 0.45 0>
     1=<1 0.4 0>
     2=<0 1 0>
     3=<0.4 -0.22 0>
     4=<0 0 1>
     5=<1 0.45 0>
     Index=$Particles.PartitionParticles.Partition
     Num=::NewEnumerator4
    )
    Particles.GearHeight:NiagaraFloat=&MergeHelper_FloatSwitch(0=10, 1=10, 2=15, 3=10, 4=0, 5=10, Index=$Particles.PartitionParticles.Partition, Num=::NewEnumerator4)
    Particles.GearPosition:Vector3f=&MergeHelper_VectorSwitch(
     0=<-80 -50 30>
     1=<-53 -76 80>
     2=<0 -111 65>
     3=<60 -62 40>
     4=<0 -40 0>
     5=<-65 -76 12>
     Index=$Particles.PartitionParticles.Partition
     Num=::NewEnumerator4
    )
    Particles.GearRadius:NiagaraFloat=&MergeHelper_FloatSwitch(0=25, 1=25, 2=44, 3=45, 4=80, 5=10, Index=$Particles.PartitionParticles.Partition, Num=::NewEnumerator4)
   ) {isSetParameters=true, module=SetVariables_#5}
   &ShapeLocation(
    Cylinder Height=$Particles.GearHeight
    Cylinder Radius=$Particles.GearRadius
    MaskSpawnGroup=false
    Offset=$Particles.GearPosition
    Offset Mode=::NewEnumerator0
    Override Seed=false
    Rotation Axis=$Particles.GearAxis
    Rotation Mode=::NewEnumerator1
    Shape Primitive=::NewEnumerator1
    Surface Only=false
    UseOwnerScale=true
   )
   &CameraOffset(Camera Offset Amount=20)
   /SetVariables(Particles.GearGlowPartition:NiagaraInt32=&ModuloInt(A=&ReturnParticleIDAttributes(), B=8)) {isSetParameters=true}
   /SetVariables(
    Particles.AngleFade:Vector4f=<1 1 0 0>
    Particles.GearGlowColor:LinearColor=<rgba 0.2791 0.2308 1 0.5>
    Particles.GearGlowFacing:Vector3f=&MergeHelper_VectorSwitch(
     0=<-1 0.5 0>
     1=<-0.5 1 0>
     2=<0 1 0>
     3=<1 0.5 0>
     4=<0.6 1 0>
     5=<1 0.5 0>
     6=<1 0.45 0>
     7=<1 0.45 0>
     Index=$Particles.GearGlowPartition
     Num=::NewEnumerator6
    )
    Particles.GearGlowPosition:NiagaraPosition=&MergeHelper_VectorSwitch(
     0=<62 -63 42>
     1=<33 -103 53>
     2=<0 -117 63>
     3=<-53 -76 78>
     4=<-33 -100 74>
     5=<-64 -54 53.5>
     6=<-80 -53 29>
     7=<-65 -79 13.5>
     Index=$Particles.GearGlowPartition
     Num=::NewEnumerator6
    )
    Particles.GearGlowSize:Vector2f=&Vector2DFromFloat(Value=&MergeHelper_FloatSwitch(0=95, 1=28, 2=97, 3=59, 4=28, 5=28, 6=60, 7=27, Index=$Particles.GearGlowPartition, Num=::NewEnumerator6))
    Particles.Lifetime=&Multiply_Float(
     A=$Particles.Lifetime
     B=&MaskFloat(Mask Source=1, Mask Parameter=$Particles.VisibilityTag, Mask=0, Masked Float=1, Passthrough Float=5000)
    )
   ) {isSetParameters=true}
  ]
  particleUpdateScript [
   &ParticleState()
   &ScaleColor(
    Scale Alpha=&FloatFromCurve(
     FloatCurve=DataInterface() {propertyValues {Curve {
        defaultValue=1
        keys [
         {interpMode=RCIM_Cubic, tangentMode=RCTM_User, time=0, value=0}
         {interpMode=RCIM_Cubic, tangentMode=RCTM_User, time=0.2269, value=1}
         {arriveTangent=-0.1174, interpMode=RCIM_Cubic, leaveTangent=-0.1174, tangentMode=RCTM_User, time=0.3787, value=1}
         {tangentMode=RCTM_User, time=1, value=0}
        ]
       }}}
     Scale Curve=&Waveform(Global Phase=$Particles.Age, "Waveform [1]"=::NewEnumerator0, "[1] Frequency"=5, "[1] Amplitude Min/Max"=<0 1>, "[1] Bias"=0.3)
    )
    ScaleA=true
    ScaleRGB=true
   )
   /SetVariables(Particles.GearGlowColor:LinearColor=&MultiplyLinearColor(
     Linear Color=$Particles.Initial.GearGlowColor
     Scale Factor=&MakeVector4_FromFloat(Float=$System.Pulsate)
    )) {isSetParameters=true}
  ]
  rendererClasses [&NiagaraSpriteRendererProperties]
  renderers [
   {rendererClass=&NiagaraSpriteRendererProperties}
   {rendererClass=&NiagaraSpriteRendererProperties, rendererIndex=1}
   {rendererClass=&NiagaraSpriteRendererProperties, rendererIndex=2}
  ]
 }
 Ground_Sparkles {
  emitterUpdateScript [
   &EmitterState()
   &SpawnRate(SpawnRate=10, Use Spawn Probability=false)
  ]
  particleSpawnScript [
   /SetVariables(Particles.Scale:NiagaraFloat=&RandomRangeFloat(Minimum=0, Maximum=1)) {isSetParameters=true, module=SetVariables_#4}
   &InitializeParticle(
    Color Mode=::NewEnumerator0
    Lifetime=&FloatFromCurve(
     FloatCurve=DataInterface() {propertyValues {Curve {keys [<at 0 1 RCIM_Constant>,<at 0.4563 1.4 RCIM_Constant>,<at 0.857 1.1 RCIM_Constant>]}}}
     CurveIndex=$Particles.Scale
     Scale Curve=&RandomRangeFloat(Minimum=0.6, Maximum=1.1)
    )
    Sprite Size Mode=::NewEnumerator3
    Uniform Sprite Size=&FloatFromCurve(
     FloatCurve=DataInterface() {propertyValues {Curve {keys [<at 0 0.3 RCIM_Constant>,<at 0.4406 0.5 RCIM_Constant>,<at 0.8099 1 RCIM_Constant>]}}}
     CurveIndex=$Particles.Scale
     Scale Curve=40
    )
    UsePositionOffset=false
    Write Material Random=true
   )
   /SetVariables(Particles.GearAxis:Vector3f=<0 0 1>, Particles.GearHeight=1, Particles.GearPosition:Vector3f=<0 -40 20>, Particles.GearRadius=70) {isSetParameters=true, module=SetVariables_#5}
   &ShapeLocation(
    Cylinder Height=$Particles.GearHeight
    Cylinder Radius=$Particles.GearRadius
    MaskSpawnGroup=false
    Offset=$Particles.GearPosition
    Offset Coordinate Space=::Local
    Offset Mode=::NewEnumerator0
    Override Seed=false
    Rotation Axis=$Particles.GearAxis
    Rotation Mode=::NewEnumerator1
    Shape Primitive=::NewEnumerator1
    Surface Only=false
    UseOwnerScale=true
   )
   /SetVariables(
    Particles.FEPosition:NiagaraPosition=&Subtract_Vector(A=Linked() {linkedVariable {name=Particles.ShapeLocation.ShapePosition, type=NiagaraPosition}}, B=<0 0 17>)
    Particles.FERotation:Vector3f=<0 0 1>
    Particles.FESize:Vector2f=&Multiply_Vector2D_ByFloat(Float=1.5, Vector2D=$Particles.SpriteSize)
   ) {isSetParameters=true}
  ]
  particleUpdateScript [
   &CameraOffset(Camera Offset Amount=20)
   &ParticleState()
   &ScaleColor(
    Scale Alpha=&FloatFromCurve(
     FloatCurve=DataInterface() {propertyValues {Curve {
        defaultValue=1
        keys [
         {interpMode=RCIM_Cubic, tangentMode=RCTM_User, time=0, value=0}
         {interpMode=RCIM_Cubic, tangentMode=RCTM_User, time=0.2269, value=1}
         {arriveTangent=-0.1174, interpMode=RCIM_Cubic, leaveTangent=-0.1174, tangentMode=RCTM_User, time=0.3787, value=1}
         {tangentMode=RCTM_User, time=1, value=0}
        ]
       }}}
     Scale Curve=&Waveform(Global Phase=$Particles.Age, "Waveform [1]"=::NewEnumerator0, "[1] Frequency"=5, "[1] Amplitude Min/Max"=<0 1>, "[1] Bias"=0.3)
    )
    Scale RGB=<1.5 1.5 1.5>
    ScaleA=true
    ScaleRGB=true
   )
   &self:Slope_Position_02(
    Apply Rotation=true
    Fallback Position:NiagaraPosition=$Particles.Initial.ShapeLocation.ShapePosition
    Height Adjustment/Offset=1
    Slope Position:NiagaraPosition=$Particles.Initial.ShapeLocation.ShapePosition
    Trace Length=120
    Trace Origin Height=85
    bIsFrontEnd:NiagaraBool=$User.bIsFrontEndPreview
   ) {module=Slope_Position_02}
  ]
  rendererClasses [&NiagaraSpriteRendererProperties]
  renderers [
   {rendererClass=&NiagaraSpriteRendererProperties}
   {rendererClass=&NiagaraSpriteRendererProperties, rendererIndex=1}
  ]
 }
 Hand_Trail_Sparkles {
  emitterUpdateScript [
   &EmitterState()
   &SpawnRate(SpawnRate=$User.HandSparkleSpawnRate, Use Spawn Probability=false)
   &SpawnRate(SpawnGroup=1, SpawnRate=$User.HandSparkleSpawnRate, Use Spawn Probability=false) {module=SpawnRate001}
  ]
  particleSpawnScript [
   /SetVariables(Particles.Scale:NiagaraFloat=&RandomRangeFloat(Minimum=0, Maximum=1)) {isSetParameters=true, module=SetVariables_#4}
   &PartitionParticles(
    A:NiagaraFloat=&RandomRangeFloat(Minimum=0, Maximum=1)
    B:NiagaraFloat=&MakeCustomFloatFromBool(
     Bool=&Boolean_NotOperation(Bool=&SetBoolByFloatComparison(A=$User.SparkleLength, B=0.99))
     True Float=1
     False Float=0.96
    )
    Partition Mode:ENiagara_PartitionMode_v2=::NewEnumerator4
   )
   &InitializeParticle(
    Color Mode=::NewEnumerator0
    Lifetime=&FloatFromCurve(
     FloatCurve=DataInterface() {propertyValues {Curve {keys [<at 0.0027 0.3 RCIM_Constant>,<at 0.5 0.8 RCIM_Constant>,<at 0.8 1.7 RCIM_Constant>]}}}
     CurveIndex=$Particles.Scale
     Scale Curve=&Multiply_Float(A=&RandomRangeFloat(Minimum=0.1, Maximum=0.3), B=$User.SparkleLength)
    )
    Sprite Size Mode=::NewEnumerator3
    Uniform Sprite Size=&FloatFromCurve(
     FloatCurve=DataInterface() {propertyValues {Curve {keys [<at 0 0.15 RCIM_Constant>,<at 0.5 0.25 RCIM_Constant>,<at 0.8 0.4 RCIM_Constant>]}}}
     CurveIndex=$Particles.Scale
     Scale Curve=&Multiply_Float(
      A=10
      B=&MergeHelper_FloatSwitch(
       0=1
       1=&RandomRangeFloat(Minimum=3, Maximum=4)
       2=?
       3=?
       Index=$Particles.PartitionParticles.Partition
       Num=::NewEnumerator0
      )
     )
    )
    UsePositionOffset=false
    Write Material Random=true
   )
   &CameraOffset(Camera Offset Amount=20)
   &SkeletalMeshLocation(
    Bone / Socket Index=$Emitter.SpawnGroup
    Bone Sampling Mode=::NewEnumerator4
    MaskSpawnGroup=false
    Skeletal Mesh=$System.Skeletal Mesh Reader
   )
   &ShapeLocation(
    MaskSpawnGroup=false
    Override Seed=false
    Shape Origin=$Particles.SkeletalMeshLocation.SampledPosition
    Sphere Radius=&Multiply_Float(
     A=6
     B=&MergeHelper_FloatSwitch(0=1, 1=0.5, 2=?, 3=?, Index=$Particles.PartitionParticles.Partition, Num=::NewEnumerator0)
    )
    UseOwnerScale=true
   )
   &InheritVelocity(
    Inherited Velocity Amount Scale=&RandomRangeVector(Minimum=<0.15 0.15 0.15>, Maximum=<0.05 0.05 0.05>)
    Speed Limit=true
    UseSpeedThreshold=false
    Velocity Source=$Particles.SkeletalMeshLocation.SampledVelocity
   )
   &ApplyInitialForces()
   /SetVariables(Particles.Lifetime=&Multiply_Float(
     A=$Particles.Lifetime
     B=&MergeHelper_FloatSwitch(0=1, 1=1.5, 2=?, 3=?, Index=$Particles.PartitionParticles.Partition, Num=::NewEnumerator0)
    )) {isSetParameters=true}
  ]
  particleUpdateScript [
   &ParticleState()
   &ScaleColor(
    Scale Alpha=&FloatFromCurve(
     FloatCurve=DataInterface() {propertyValues {Curve {
        defaultValue=1
        keys [
         {interpMode=RCIM_Cubic, tangentMode=RCTM_User, time=0, value=1}
         {arriveTangent=-0.1174, interpMode=RCIM_Cubic, leaveTangent=-0.1174, tangentMode=RCTM_User, time=0.2793, value=1}
         {tangentMode=RCTM_User, time=1, value=0}
        ]
       }}}
     Scale Curve=&Waveform(Global Phase=$Particles.Age, "Waveform [1]"=::NewEnumerator0, "[1] Frequency"=5, "[1] Amplitude Min/Max"=<0 1>, "[1] Bias"=0.3)
    )
    ScaleA=true
    ScaleRGB=true
   )
   &Drag(Use Linear Drag=true, Use Rotational Drag=true)
   &SolveForcesAndVelocity(Clamp Velocity=false, Limit Acceleration=false)
  ]
  rendererClasses [&NiagaraSpriteRendererProperties]
  renderers [{rendererClass=&NiagaraSpriteRendererProperties}]
 }
}
summary {
 systemName=NS_IndoorLaceBeat_Gears
 emitters [
  {emitterName=Hand_Trail_Sparkles, rendererClasses [&NiagaraSpriteRendererProperties]}
  {emitterName=Gear_Sparkles, rendererClasses [&NiagaraSpriteRendererProperties]}
  {
   emitterName=Clock_and_Lights
   rendererClasses [&NiagaraDecalRendererProperties, &NiagaraLightRendererProperties]
  }
  {emitterName=Ground_Sparkles, rendererClasses [&NiagaraSpriteRendererProperties]}
  {emitterName=FE_Clock, rendererClasses [&NiagaraSpriteRendererProperties, &NiagaraLightRendererProperties]}
 ]
 userVariables [
  User.bIsFrontEnd {type=NiagaraBool}
  User.bIsFrontEndPreview {type=NiagaraBool}
  User.HandSparkleSpawnRate {type=NiagaraFloat, defaultValue=250, description=""}
  User.SparkleLength {type=NiagaraFloat, defaultValue=1, description=""}
 ]
}
